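# Parent material for the male medium RankedPhantom character body cosmetic, built on material attributes with the shading model chosen by expression.
# Shared 2023 character PreFX and PostFX functions bracket a ranked chain of thin film, rank icon atlas and emissive Fresnel, using the Tech Art Mask.
# A final pass tints base color and biases metallic and roughness, blended in through a selectable skull mask channel.
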Material /BRCosmetics/Characters/Player/Male/Medium/Bodies/M_MED_RankedPhantom/Materials/M_RankedPhantom_Parent
graph {
 properties {
  bAutomaticallySetUsageInEditor=false
  bUseMaterialAttributes=true
  bUsedWithMeshDeformer=true
  bUsedWithMorphTargets=true
  bUsedWithSkeletalMesh=true
  pixelDepthOffsetMode=PDOM_Legacy
  shadingModel=MSM_FromMaterialExpression
  usedShadingModels=Default Lit | Subsurface | Clear Coat | Subsurface Profile | Cloth
 }
 nodes [
  n1 = MaterialFunctionCall() {materialFunction=&MF_CharacterParent_2023_PreFX}
  MaterialFunctionCall(Material=NamedRerouteUsage(), M=@n1.M, Pre Skinned Local Position=@n1.M, Distance Blend=@n1.M, Fresnel=@n1.M) {materialFunction=&MF_CharacterParent_2023_PostFX}
  NamedRerouteDeclaration(None=@n1.Tech Art Mask) {name=TAMask, nodeColor=<rgba 1 0.8672 0 1>}
  n2 = MaterialFunctionCall!(
   Material=MaterialFunctionCall!(Material Attributes=@n1.Result, Alpha=@n1.Result) {materialFunction=/Game/Global/Materials/Cosmetics/Ranked/MF_RankedThinFilm}.Result
   Alpha=@n1.Tech Art Mask
  ) {materialFunction=/Game/Global/Materials/Cosmetics/Ranked/MF_RankIconAtlas}
  NamedRerouteDeclaration(None=MaterialFunctionCall!(Material=@n2.Result, Mask=@n1.Tech Art Mask, Fresnel=@n1.Tech Art Mask) {materialFunction=/Game/Global/Materials/Cosmetics/Ranked/MF_RankedEmissiveFresnel}.Result) {name=MA-Pre, nodeColor=<rgba 0.5469 0 1 1>}
  n3 = NamedRerouteUsage()
  n4 = GetMaterialAttributes(None=@n3) {attributeGetTypes [
    #1
    #2
    #3
    #4
   ]}
  n5 = SetMaterialAttributes!(
   MaterialAttributes=@n4.MaterialAttributes
   Base Color=Multiply(A=@n4.Base Color, B=VectorParameter() {group=RankedPhantom, parameterName=Ranked_BaseColor_Tint, defaultValue=<rgba 1 1 1 1>}.RGB)
   Metallic=Add(A=@n4.Metallic, B=ScalarParameter() {group=RankedPhantom, parameterName=Ranked_Metallic_Bias, sliderMax=1, sliderMin=-1})
   Specular=@n4.MaterialAttributes
   Roughness=Add(A=@n4.Roughness, B=ScalarParameter() {group=RankedPhantom, parameterName=Ranked_Roughness_Bias, sliderMax=1, sliderMin=-1})
  ) {attributeSetTypes [
    #1
    #2
    #3
    #4
   ]}
  n6 = NamedRerouteUsage()
  NamedRerouteDeclaration(None=BlendMaterialAttributes(A=@n3, B=@n5, Alpha=ChannelMaskParameter(None=@n6) {group=RankedPhantom, parameterName=SkullMaskChannel}.R)) {name=MA-Post, nodeColor=<rgba 0 1 0.2266 1>}
  Reroute(None=@n2.Result)
  Reroute(None=@n3)
 ]
}
