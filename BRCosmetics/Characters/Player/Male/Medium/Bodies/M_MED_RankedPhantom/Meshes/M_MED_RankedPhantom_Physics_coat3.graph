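# Physics asset for the RankedPhantom medium male player body that sets up a simulated long coat with lower-body colliders.
# Five three-segment capsule chains (back, sides, two fronts) hang from small base bodies with limited swing, and twist is locked on the back.
# Free-rotation constraints tie neighboring hem segments together, leaving the front open.
# Pelvis and leg capsules provide body shapes, and a huge offset sphere on dyn_flur apparently approximates the floor.
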PhysicsAsset /BRCosmetics/Characters/Player/Male/Medium/Bodies/M_MED_RankedPhantom/Meshes/M_MED_RankedPhantom_Physics_coat3
bodies {
 calf_l [{shapeType=Capsule, length=38.6808, radius=10.3192, center=<-19.7604 -0.923 3.6235>, rotation=<rot 87.339 0 2.9674>}]
 calf_r [{shapeType=Capsule, length=39.4788, radius=9.5212, center=<19.8395 0.923 -1.9223>, rotation=<rot 87.339 0 -176.8509>}]
 coat_bck_01 [{
   shapeType=Capsule
   length=28.7309
   radius=10
   shapeName=coat_fwd1_l_03_capsule
   center=<10.1674 -9.4237 -0.2667>
   rotation=<rot 0.016 -3.0479 0.1478>
  }]
 coat_bck_02 [{shapeType=Capsule, length=31.91, radius=10, shapeName=coat_fwd1_l_03_capsule, center=<8.373 -7.8885 -0.2709>, rotation=<rot 0 3.1496 0.1486>}]
 coat_bck_03 [{
   shapeType=Capsule
   length=30.0695
   radius=10
   shapeName=coat_fwd1_l_03_capsule
   center=<0.7389 -7.1677 0.005>
   rotation=<rot 0 -1.1632 0.3065>
  }]
 coat_bck_base [{shapeType=Capsule, length=0.505, radius=0.505, shapeName=coat_bck_base_capsule, rotation=<rot 0 -90 0>}]
 coat_fwd1_l_01 [{
   shapeType=Capsule
   length=15.0335
   radius=7.9665
   shapeName=coat_fwd1_l_01_capsule
   center=<14.9618 -6.8522 1.0746>
   rotation=<rot -2.0558 176.5126 172.5568>
  }]
 coat_fwd1_l_02 [{
   shapeType=Capsule
   length=15.8771
   radius=9.9738
   shapeName=coat_fwd1_l_02_capsule
   center=<10.9879 -8.895 0.6617>
   rotation=<rot 1.3535 0.4903 -6.2481>
  }]
 coat_fwd1_l_03 [{
   shapeType=Capsule
   length=15.609
   radius=10
   shapeName=coat_fwd1_l_03_capsule
   center=<2.0564 -7.9695 2.185>
   rotation=<rot 0.3852 0.1715 15.9253>
  }]
 coat_fwd1_l_base [{shapeType=Capsule, length=0.505, radius=0.505, shapeName=coat_fwd1_l_base_capsule, rotation=<rot 0 -90 0>}]
 coat_fwd1_r_01 [{
   shapeType=Capsule
   length=15.0335
   radius=7.9665
   shapeName=coat_fwd1_l_01_capsule
   center=<14.9618 -6.8522 1.0746>
   rotation=<rot -2.0558 176.5126 172.5568>
  }]
 coat_fwd1_r_02 [{
   shapeType=Capsule
   length=15.8771
   radius=9.9738
   shapeName=coat_fwd1_l_02_capsule
   center=<10.9879 -8.895 0.6617>
   rotation=<rot 1.3535 0.4903 -6.2481>
  }]
 coat_fwd1_r_03 [{
   shapeType=Capsule
   length=15.609
   radius=10
   shapeName=coat_fwd1_l_03_capsule
   center=<3.3776 -8.3808 2.1567>
   rotation=<rot 0.3852 0.1714 -5.1113>
  }]
 coat_fwd1_r_base [{shapeType=Capsule, length=0.505, radius=0.505, shapeName=coat_fwd1_r_base_capsule, rotation=<rot 0 -90.0001 2.8126>}]
 coat_l_01 [{
   shapeType=Capsule
   length=15.0335
   radius=7.9665
   shapeName=coat_fwd1_r_01_capsule
   center=<11.9878 -5.8036 4.3402>
   rotation=<rot -1.2498 -1.408 3.4409>
  }]
 coat_l_02 [{
   shapeType=Capsule
   length=31.0745
   radius=10
   shapeName=coat_fwd1_r_03_capsule
   center=<6.7387 -8.6893 -3.7422>
   rotation=<rot -1.3686 10.366 -4.4986>
  }]
 coat_l_03 [{
   shapeType=Capsule
   length=32.1615
   radius=10
   shapeName=coat_fwd1_l_03_capsule
   center=<1.028 -7.672 -2.7902>
   rotation=<rot 1.9242 8.7169 -2.4419>
  }]
 coat_l_base [{shapeType=Capsule, length=0.505, radius=0.505, shapeName=coat_l_base_capsule, rotation=<rot 0 -90 0>}]
 coat_r_01 [{
   shapeType=Capsule
   length=15.0335
   radius=7.9665
   shapeName=coat_fwd1_r_01_capsule
   center=<11.9878 -5.8036 -4.3402>
   rotation=<rot -1.2498 -1.408 -3.4409>
  }]
 coat_r_02 [{
   shapeType=Capsule
   length=31.0745
   radius=10
   shapeName=coat_fwd1_r_03_capsule
   center=<6.7387 -8.6893 3.7422>
   rotation=<rot 1.3686 10.366 -4.4986>
  }]
 coat_r_03 [{
   shapeType=Capsule
   length=32.1615
   radius=10
   shapeName=coat_fwd1_l_03_capsule
   center=<1.0697 -7.9465 2.7784>
   rotation=<rot -1.9242 8.7169 -2.4419>
  }]
 coat_r_base [{shapeType=Capsule, length=0.505, radius=0.505, shapeName=coat_r_base_capsule, rotation=<rot 0 -90 0>}]
 dyn_flur [{shapeType=Sphere, radius=135.8838, shapeName=dyn_flur_sphere, center=<0 0 -129.9846>}]
 foot_l [{shapeType=Capsule, length=12.834, radius=9.4884, center=<-5.5 5.5 1.7458>, rotation=<rot 0 -0.0001 90>}]
 foot_r [{shapeType=Capsule, length=12.3304, radius=9.992, center=<5.5 -5.5 -2.4694>, rotation=<rot 0 -0.0001 90>}]
 pelvis [{shapeType=Capsule, length=8.1019, radius=10.5178, center=<-1.5007 -1.3221 0>}]
 thigh_l [{shapeType=Capsule, length=28.3109, radius=9.689, center=<-21.016 -1.3658 0.5372>, rotation=<rot 90 -0.3669 0>}]
 thigh_r [{shapeType=Capsule, length=28.3109, radius=9.689, center=<21.016 1.3658 -0.5372>, rotation=<rot 90 -0.3669 0>}]
}
constraints [
 {bone1Name=coat_bck_01, bone2Name=coat_bck_base, swing1LimitDegrees=15, swing2LimitDegrees=15, twistLimitDegrees=5, twistMotion=Locked}
 {bone1Name=coat_bck_02, bone2Name=coat_bck_01, swing1LimitDegrees=15, swing2LimitDegrees=15, twistLimitDegrees=5, twistMotion=Locked}
 {bone1Name=coat_bck_03, bone2Name=coat_bck_02, swing1LimitDegrees=20, swing2LimitDegrees=10, twistLimitDegrees=5, twistMotion=Locked}
 {bone1Name=coat_fwd1_l_01, bone2Name=coat_fwd1_l_base, swing1LimitDegrees=20, swing2LimitDegrees=10, twistLimitDegrees=5}
 {bone1Name=coat_fwd1_l_02, bone2Name=coat_fwd1_l_01, swing1LimitDegrees=30, swing2LimitDegrees=10, twistLimitDegrees=10}
 {bone1Name=coat_fwd1_l_03, bone2Name=coat_fwd1_l_02, swing1LimitDegrees=25, swing2LimitDegrees=10, twistLimitDegrees=5}
 {bone1Name=coat_fwd1_r_01, bone2Name=coat_fwd1_r_base, swing1LimitDegrees=20, swing2LimitDegrees=10, twistLimitDegrees=5}
 {bone1Name=coat_fwd1_r_02, bone2Name=coat_fwd1_r_01, swing1LimitDegrees=15, swing2LimitDegrees=10, twistLimitDegrees=10}
 {bone1Name=coat_fwd1_r_03, bone2Name=coat_fwd1_r_02, swing1LimitDegrees=20, swing2LimitDegrees=10, twistLimitDegrees=5}
 {bone1Name=coat_l_01, bone2Name=coat_l_base, swing1LimitDegrees=20, swing2LimitDegrees=20, twistLimitDegrees=10}
 {bone1Name=coat_l_02, bone2Name=coat_l_01, swing1LimitDegrees=30, swing2LimitDegrees=10, twistLimitDegrees=5}
 {bone1Name=coat_l_03, bone2Name=coat_l_02, swing1LimitDegrees=20, swing2LimitDegrees=10, twistLimitDegrees=5}
 {bone1Name=coat_r_01, bone2Name=coat_r_base, swing1LimitDegrees=20, swing2LimitDegrees=20, twistLimitDegrees=10}
 {bone1Name=coat_r_02, bone2Name=coat_r_01, swing1LimitDegrees=30, swing2LimitDegrees=10, twistLimitDegrees=5}
 {bone1Name=coat_r_03, bone2Name=coat_r_02, swing1LimitDegrees=20, swing2LimitDegrees=10, twistLimitDegrees=5}
 {
  bone1Name=coat_r_03
  bone2Name=coat_fwd1_r_03
  swing1LimitDegrees=45
  swing1Motion=Free
  swing2LimitDegrees=45
  swing2Motion=Free
  twistLimitDegrees=45
  twistMotion=Free
 }
 {
  bone1Name=coat_bck_03
  bone2Name=coat_l_03
  swing1LimitDegrees=45
  swing1Motion=Free
  swing2LimitDegrees=45
  swing2Motion=Free
  twistLimitDegrees=45
  twistMotion=Free
 }
 {
  bone1Name=coat_bck_03
  bone2Name=coat_r_03
  swing1LimitDegrees=45
  swing1Motion=Free
  swing2LimitDegrees=45
  swing2Motion=Free
  twistLimitDegrees=45
  twistMotion=Free
 }
 {
  bone1Name=coat_l_03
  bone2Name=coat_fwd1_l_03
  swing1LimitDegrees=45
  swing1Motion=Free
  swing2LimitDegrees=45
  swing2Motion=Free
  twistLimitDegrees=45
  twistMotion=Free
 }
]
