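# Control Rig for the Ranked Phantom medium male body that corrects coat and body deformation as the character moves.
# A shared DeformRig function runs first, then constraints blend coat bases between pelvis and thighs.
# Leg ROM outputs, foot-to-pelvis distances and slide, parachute and skydive flags weight additional coat offsets.
# A right upper-arm twist control drives shoulder correctives with pose-reader adjustments, and foot yaw drives foot correctives.
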ControlRigBlueprint /BRCosmetics/Characters/Player/Male/Medium/Bodies/M_MED_RankedPhantom/Meshes/M_MED_RankedPhantom_cape
graphs [
 Multiply_10_ContainedGraph {nodes [
   n1 = RigVMFunctionEntryNode()
   RigVMFunctionReturnNode(Result=&RigVMFunction_MathDoubleMul(B=@n1.C, A=&RigVMFunction_MathDoubleMul(A=@n1.A, B=@n1.B).Result).Result)
  ]}
 RigVMFunctionLibrary {links [], nodes []}
 RigVMFunction_Sequence_1_ContainedGraph {nodes [
   n1 = RigVMFunctionEntryNode()
   RigVMFunction_Sequence = &RigVMFunction_Sequence(ExecuteContext=@n1.ExecuteContext)
   n2 = &RigVMFunction_Sequence(ExecuteContext=@RigVMFunction_Sequence.B)
   n3 = &RigVMFunction_Sequence(ExecuteContext=@n2.B)
   RigVMFunctionReturnNode(A=@RigVMFunction_Sequence.A, B=@n2.A, C=@n3.A, D=@n3.B)
  ]}
 RigVMModel {nodes [
   n1 = &RigVMFunction_MathBoolToFloat(Value=RigVMVariableNode() {variable=isDBNO}.Value)
   &RigUnit_ModifyTransforms(
    ExecutePin=()
    ItemToModify="((Item=(Type=Bone,Name=\"coat_bck_base\"),Transform=(Rotation=(X=0,Y=0,Z=0.342,W=0.9397),Translation=(X=1,Y=-1,Z=0),Scale3D=(X=1,Y=1,Z=1))),(Item=(Type=Bone,Name=\"coat_fwd1_l_base\"),Transform=(Rotation=(X=0.0226,Y=-0.2578,Z=0.0842,W=0.9623),Translation=(X=0, ...(truncated)"
    WeightMinimum=0
    Weight=@n1.Result
   )
   > n2 = &RigUnit_SphericalPoseReader(
    ActiveRegionScaleFactors=(PositiveWidth=1,NegativeWidth=1.5,PositiveHeight=1.5,NegativeHeight=1)
    ActiveRegionSize=0.1
    Debug=(bDrawDebug=True,bDraw2D=False,bDrawLocalAxes=False,DebugScale=25,DebugSegments=20,DebugThickness=0.25)
    DriverAxis=(X=1,Y=0,Z=0)
    DriverItem=(Type=Bone,Name="thigh_r")
    FalloffRegionScaleFactors=(PositiveWidth=0.4,NegativeWidth=1,PositiveHeight=1,NegativeHeight=1)
    FalloffSize=0.3
    FlipHeightScaling=False
    FlipWidthScaling=False
    OptionalParentItem=(Type=Bone,Name="None")
    RotationOffset=(X=-35,Y=-30,Z=0)
   )
   > &RigUnit_ModifyTransforms(
    ItemToModify="((Item=(Type=Bone,Name=\"coat_fwd1_r_base\"),Transform=(Rotation=(X=0,Y=0,Z=-0.1736,W=0.9848),Translation=(X=0,Y=0,Z=0),Scale3D=(X=1,Y=1,Z=1))),(Item=(Type=Bone,Name=\"coat_r_base\"),Transform=(Rotation=(X=-0.0151,Y=-0.173,Z=-0.0858,W=0.9811),Translation=(X=0, ...(truncated)"
    WeightMinimum=0
    Weight=&RigVMFunction_MathDoubleMul(A=@n1.Result, B=@n2.OutputParam).Result
   )
   > n3 = &RigUnit_SphericalPoseReader(
    ActiveRegionScaleFactors=(PositiveWidth=1,NegativeWidth=1.5,PositiveHeight=1.5,NegativeHeight=1)
    ActiveRegionSize=0.1
    Debug=(bDrawDebug=True,bDraw2D=False,bDrawLocalAxes=False,DebugScale=25,DebugSegments=20,DebugThickness=0.25)
    DriverAxis=(X=-1,Y=0,Z=0)
    DriverItem=(Type=Bone,Name="thigh_l")
    FalloffRegionScaleFactors=(PositiveWidth=0.4,NegativeWidth=1,PositiveHeight=1,NegativeHeight=1)
    FalloffSize=0.3
    FlipHeightScaling=false
    FlipWidthScaling=true
    OptionalParentItem=(Type=Bone,Name="None")
    RotationOffset=(X=145,Y=-30,Z=0)
   )
   > &RigUnit_ModifyTransforms(
    ItemToModify="((Item=(Type=Bone,Name=\"coat_fwd1_l_base\"),Transform=(Rotation=(X=0,Y=0,Z=-0.0872,W=0.9962),Translation=(X=0,Y=0,Z=0),Scale3D=(X=1,Y=1,Z=1))),(Item=(Type=Bone,Name=\"coat_l_base\"),Transform=(Rotation=(X=0.0151,Y=0.173,Z=-0.0858,W=0.9811),Translation=(X=0,Y= ...(truncated)"
    WeightMinimum=0
    Weight=&RigVMFunction_MathDoubleMul(A=@n1.Result, B=@n3.OutputParam).Result
   )
   RigUnit_BeginExecution = &RigUnit_BeginExecution()
   &RigUnit_ParentConstraint(
    AdvancedSettings=(InterpolationType=Average,RotationOrderForFilter=XZY)
    Child=(Type=Bone,Name="shoulder_cor2")
    Filter=(TranslationFilter=(bX=True,bY=True,bZ=True),RotationFilter=(bX=false,bY=false,bZ=false),ScaleFilter=(bX=True,bY=True,bZ=True))
    Parents=((Item=(Type=Bone,Name="deform_upperarm_bk_base_r"),Weight=1))
    Weight=1
    ExecutePin=@RigUnit_BeginExecution.ExecutePin
   )
   > &RigUnit_ModifyTransforms(
    ItemToModify="((Item=(Type=Bone,Name=\"thigh_l\"),Transform=(Rotation=(X=-0.183,Y=-0.183,Z=-0.683,W=0.683),Translation=(X=0,Y=0,Z=0),Scale3D=(X=1,Y=1,Z=1))),(Item=(Type=Bone,Name=\"thigh_r\"),Transform=(Rotation=(X=-0.183,Y=-0.183,Z=-0.683,W=0.683),Translation=(X=0,Y=0,Z=0) ...(truncated)"
    Weight=0
    WeightMinimum=0
   )
   > FN_DeformRig_FNC = RigVMFunctionReferenceNode(
    DeformRig_ Curve="(deformRig_l_upperarm_up_alpha_72_#1=False,deformRig_r_upperarm_up_alpha_73_#2=False,deformRig_l_upperarm_fr_alpha_74_#3=False,deformRig_r_upperarm_fr_alpha_75_CB51B8 ...(truncated)"
    DeformRig_Bool="(upperarm_1_#4=True,elbow_3_#5=True,wrist_7_#6=True,groin_8_#7=True,glute_17_#8=True,knee_9_40C408324CA261 ...(truncated)"
    DeformRig_Debug=(upperarm_debug_18_#4=False,wrist_debug_20_#6=False,lat_debug_26_#9=False,groinGlute_debug_27_#7=False)
    DeformRig_Value="(upperarm_up_18_#10=(Rotation=(X=0,Y=0,Z=0,W=1),Translation=(X=0,Y=0,Z=8),Scale3D=(X=0.7,Y=1,Z=1)),upperarm_fwd_fr_19_#11=(Rotation=(X=0,Y=0,Z=0,W=1),Translation=(X=0,Y=10,Z=0),Scale3D=(X=0.5,Y=1,Z= ...(truncated)"
   )
   > &RigUnit_ModifyTransforms(
    ItemToModify=((Item=(Type=Bone,Name="thigh_l"),Transform=(Rotation=(X=0,Y=0,Z=0,W=1),Translation=(X=0,Y=0,Z=0),Scale3D=(X=1,Y=1,Z=1))),(Item=(Type=Bone,Name="thigh_r"),Transform=(Rotation=(X=0,Y=0,Z=0,W=1),Translation=(X=0,Y=0,Z=0),Scale3D=(X=1,Y=1,Z=1))))
    Weight=1
    WeightMinimum=0
   )
   > &RigUnit_ParentConstraint(
    AdvancedSettings=(InterpolationType=Average,RotationOrderForFilter=XZY)
    Child=(Type=Bone,Name="coat_l_base")
    Filter=(TranslationFilter=(bX=false,bY=false,bZ=false),RotationFilter=(bX=true,bY=true,bZ=true),ScaleFilter=(bX=true,bY=true,bZ=true))
    Parents=((Item=(Type=Bone,Name="thigh_l"),Weight=1),(Item=(Type=Bone,Name="pelvis"),Weight=1))
    Weight=1
   )
   > &RigUnit_ParentConstraint(
    AdvancedSettings=(InterpolationType=Average,RotationOrderForFilter=XZY)
    Child=(Type=Bone,Name="coat_r_base")
    Filter=(TranslationFilter=(bX=false,bY=false,bZ=false),RotationFilter=(bX=true,bY=true,bZ=true),ScaleFilter=(bX=true,bY=true,bZ=true))
    Parents=((Item=(Type=Bone,Name="thigh_r"),Weight=1),(Item=(Type=Bone,Name="pelvis"),Weight=1))
    Weight=1
   )
   > &RigUnit_ParentConstraint(
    AdvancedSettings=(InterpolationType=Average,RotationOrderForFilter=XZY)
    Child=(Type=Bone,Name="coat_bck_base")
    Filter=(TranslationFilter=(bX=false,bY=false,bZ=false),RotationFilter=(bX=true,bY=true,bZ=true),ScaleFilter=(bX=true,bY=true,bZ=true))
    Parents=((Item=(Type=Bone,Name="thigh_l"),Weight=0.1),(Item=(Type=Bone,Name="thigh_r"),Weight=0.1),(Item=(Type=Bone,Name="pelvis"),Weight=1))
    Weight=1
   )
   > &RigUnit_ParentConstraint(
    AdvancedSettings=(InterpolationType=Average,RotationOrderForFilter=XZY)
    Child=(Type=Bone,Name="coat_fwd1_l_base")
    Filter=(TranslationFilter=(bX=false,bY=false,bZ=false),RotationFilter=(bX=true,bY=true,bZ=true),ScaleFilter=(bX=true,bY=true,bZ=true))
    Parents=((Item=(Type=Bone,Name="thigh_l"),Weight=1),(Item=(Type=Bone,Name="pelvis"),Weight=0.1))
    Weight=1
   )
   > &RigUnit_ParentConstraint(
    AdvancedSettings=(InterpolationType=Average,RotationOrderForFilter=XZY)
    Child=(Type=Bone,Name="coat_fwd1_r_base")
    Filter=(TranslationFilter=(bX=false,bY=false,bZ=false),RotationFilter=(bX=true,bY=true,bZ=true),ScaleFilter=(bX=true,bY=true,bZ=true))
    Parents=((Item=(Type=Bone,Name="thigh_r"),Weight=1),(Item=(Type=Bone,Name="pelvis"),Weight=0.1))
    Weight=1
   )
   > &RigUnit_ParentConstraint(
    AdvancedSettings=(InterpolationType=Average,RotationOrderForFilter=XZY)
    Child=(Type=Bone,Name="coat_fwd1_l_base")
    Filter=(TranslationFilter=(bX=true,bY=true,bZ=true),RotationFilter=(bX=false,bY=false,bZ=false),ScaleFilter=(bX=false,bY=false,bZ=false))
    Parents=((Item=(Type=Bone,Name="thigh_l"),Weight=0),(Item=(Type=Bone,Name="pelvis"),Weight=1))
    Weight=1
   )
   > &RigUnit_ParentConstraint(
    AdvancedSettings=(InterpolationType=Average,RotationOrderForFilter=XZY)
    Child=(Type=Bone,Name="coat_fwd1_r_base")
    Filter=(TranslationFilter=(bX=true,bY=true,bZ=true),RotationFilter=(bX=false,bY=false,bZ=false),ScaleFilter=(bX=false,bY=false,bZ=false))
    Parents=((Item=(Type=Bone,Name="thigh_r"),Weight=0),(Item=(Type=Bone,Name="pelvis"),Weight=1))
    Weight=1
   )
   > &RigUnit_ParentConstraint(
    AdvancedSettings=(InterpolationType=Average,RotationOrderForFilter=XZY)
    Child=(Type=Bone,Name="shoulder_cor")
    Parents=((Item=(Type=Control,Name="upperarm_twist_01_r_ctrl"),Weight=1))
    Weight=1
   )
   > &RigUnit_ParentConstraint(
    AdvancedSettings=(InterpolationType=Average,RotationOrderForFilter=XZY)
    Child=(Type=Bone,Name="shoulder_cor1")
    Parents=((Item=(Type=Control,Name="upperarm_twist_01_r_ctrl"),Weight=1))
    Weight=1
   )
   > &RigUnit_ParentConstraint(
    AdvancedSettings=(InterpolationType=Average,RotationOrderForFilter=XZY)
    Child=(Type=Bone,Name="shoulder_cor2")
    Parents=((Item=(Type=Control,Name="upperarm_twist_01_r_ctrl"),Weight=1))
    Weight=1
   )
   > &RigUnit_ParentConstraint(
    AdvancedSettings=(InterpolationType=Average,RotationOrderForFilter=XZY)
    Child=(Type=Bone,Name="shoulder_cor3")
    Parents=((Item=(Type=Control,Name="upperarm_twist_01_r_ctrl"),Weight=1))
    Weight=1
   )
   > &RigUnit_ParentConstraint(
    AdvancedSettings=(InterpolationType=Average,RotationOrderForFilter=XZY)
    Child=(Type=Bone,Name="shoulder_cor4")
    Parents=((Item=(Type=Control,Name="upperarm_twist_01_r_ctrl"),Weight=1))
    Weight=1
   )
   > RigVMFunction_Sequence = RigVMAggregateNode()
   Leg ROM = RigVMFunctionReferenceNode(
    Include Ankles=false
    Include Distance Checks=false
    Include Knees=true
    Interp Output Speed=5
    Interp Results=false
    Result Alpha=1
    SPR Optional Parent Item=(Type=Bone,Name="None")
    ExecuteContext=@RigVMFunction_Sequence.A
   )
   > RigVMVariableNode(Value=@"Leg ROM".L Thigh Raised Fwd) {variable=LegForward_L}
   > RigVMVariableNode(Value=@"Leg ROM".R Thigh Raised Fwd) {variable=LegForward_R}
   > RigVMVariableNode(Value=@"Leg ROM".L Thigh Raised Fwd Side Inbetween) {variable=Leg ROM_L Thigh Raised Fwd Side Inbetween}
   > RigVMVariableNode(Value=@"Leg ROM".R Thigh Raised Fwd Side Inbetween) {variable=Leg ROM_R Thigh Raised Fwd Side Inbetween}
   > RigVMVariableNode(Value=@"Leg ROM".L Thigh Raised Side) {variable=Leg ROM_L Thigh Raised Side}
   > RigVMVariableNode(Value=@"Leg ROM".R Thigh Raised Side) {variable=Leg ROM_R Thigh Raised Side}
   > RigVMVariableNode(Value=@"Leg ROM".L Knee) {variable=Leg ROM_L Knee}
   > RigVMVariableNode(Value=@"Leg ROM".R Knee) {variable=Leg ROM_R Knee}
   > RigVMVariableNode(Value=@"Leg ROM".L Thigh Raised Rear) {variable=Leg ROM_L Thigh Raised Rear}
   > RigVMVariableNode(Value=@"Leg ROM".R Thigh Raised Rear) {variable=Leg ROM_R Thigh Raised Rear}
   > RigVMVariableNode(Value=@"Leg ROM".Both Thighs Back) {variable=Leg ROM_Both Thighs Back}
   &RigUnit_ModifyTransforms(
    ItemToModify="((Item=(Type=Bone,Name=\"coat_fwd1_l_base\"),Transform=(Rotation=(X=-0.327,Y=-0.6387,Z=-0.4196,W=-0.556),Translation=(X=-0.5,Y=-7,Z=-6.5),Scale3D=(X=1,Y=1,Z=1))),(Item=(Type=Bone,Name=\"coat_l_base\"),Transform=(Rotation=(X=0.0826,Y=0.4467,Z=-0.027,W=0.8905),T ...(truncated)"
    WeightMinimum=0
    ExecutePin=@RigVMFunction_Sequence.B
    Weight=RigVMVariableNode() {variable=LegForward_L}.Value
   )
   > &RigUnit_ModifyTransforms(
    ItemToModify="((Item=(Type=Bone,Name=\"coat_fwd1_r_base\"),Transform=(Rotation=(X=-0.3357,Y=-0.6342,Z=0.4201,W=0.5555),Translation=(X=-0.5,Y=-3.5,Z=4),Scale3D=(X=1,Y=1,Z=1))),(Item=(Type=Bone,Name=\"coat_r_base\"),Transform=(Rotation=(X=-0.1017,Y=-0.4427,Z=-0.0652,W=0.8885) ...(truncated)"
    WeightMinimum=0
    Weight=RigVMVariableNode() {variable=LegForward_R}.Value
   )
   > &RigUnit_ModifyTransforms(
    ItemToModify="((Item=(Type=Bone,Name=\"coat_fwd1_l_base\"),Transform=(Rotation=(X=0.1023,Y=0.0033,Z=-0.0442,W=0.9938),Translation=(X=-2.7405,Y=-2.866,Z=0),Scale3D=(X=1,Y=1,Z=1))),(Item=(Type=Bone,Name=\"coat_l_base\"),Transform=(Rotation=(X=0.043,Y=0.0169,Z=0.0007,W=0.9989) ...(truncated)"
    WeightMinimum=0
    Weight=&RigVMFunction_MathDoubleRemap(
     SourceMaximum=1
     SourceMinimum=0
     TargetMaximum=1
     TargetMinimum=0
     Value=RigVMVariableNode() {variable=Leg ROM_L Thigh Raised Fwd Side Inbetween}.Value
    ).Result
   )
   > &RigUnit_ModifyTransforms(
    ItemToModify="((Item=(Type=Bone,Name=\"coat_fwd1_r_base\"),Transform=(Rotation=(X=-0.1023,Y=-0.0033,Z=-0.0442,W=0.9938),Translation=(X=-2.7405,Y=-2.866,Z=0),Scale3D=(X=1,Y=1,Z=1))),(Item=(Type=Bone,Name=\"coat_r_base\"),Transform=(Rotation=(X=-0.043,Y=-0.0167,Z=-0.0022,W=0. ...(truncated)"
    WeightMinimum=0
    Weight=&RigVMFunction_MathDoubleRemap(
     SourceMaximum=1
     SourceMinimum=0.4
     TargetMaximum=1
     TargetMinimum=0
     Value=RigVMVariableNode() {variable=Leg ROM_R Thigh Raised Fwd Side Inbetween}.Value
    ).Result
   )
   n4 = &RigUnit_GetTransform(Item=(Type=Bone,Name="pelvis"), bInitial=False)
   RigVMFunction_MathVectorDistance = &RigVMFunction_MathVectorDistance(A=@n4.Transform.Translation, B=&RigUnit_GetTransform(Item=(Type=Bone,Name="foot_l"), bInitial=False).Transform.Translation)
   Remap = &RigVMFunction_MathDoubleRemap(SourceMaximum=45, SourceMinimum=80, TargetMaximum=1, TargetMinimum=0, Value=@RigVMFunction_MathVectorDistance.Result)
   > &RigUnit_ModifyTransforms(
    ItemToModify="((Item=(Type=Bone,Name=\"coat_bck_base\"),Transform=(Rotation=(X=0,Y=0,Z=0,W=1),Translation=(X=0,Y=0,Z=0),Scale3D=(X=1,Y=1,Z=1))),(Item=(Type=Bone,Name=\"coat_l_base\"),Transform=(Rotation=(X=0.0151,Y=0.173,Z=-0.0858,W=0.9811),Translation=(X=-2,Y=0,Z=-2),Scale ...(truncated)"
    WeightMinimum=0
    Weight=@Remap.Result
   )
   n5 = &RigVMFunction_MathDoubleRemap(
    SourceMaximum=45
    SourceMinimum=80
    TargetMaximum=1
    TargetMinimum=0
    Value=&RigVMFunction_MathVectorDistance(A=@n4.Transform.Translation, B=&RigUnit_GetTransform(Item=(Type=Bone,Name="foot_r"), bInitial=False).Transform.Translation).Result
   )
   > &RigUnit_ModifyTransforms(
    ItemToModify="((Item=(Type=Bone,Name=\"coat_bck_base\"),Transform=(Rotation=(X=0,Y=0,Z=0,W=1),Translation=(X=0,Y=0,Z=0),Scale3D=(X=1,Y=1,Z=1))),(Item=(Type=Bone,Name=\"coat_r_base\"),Transform=(Rotation=(X=-0.0151,Y=-0.173,Z=-0.0858,W=0.9811),Translation=(X=0,Y=0,Z=2),Scale ...(truncated)"
    WeightMinimum=0
    Weight=@n5.Result
   )
   Add = &RigVMFunction_MathDoubleAdd(A=@Remap.Result, B=@n5.Result)
   Clamp = &RigVMFunction_MathDoubleClamp(Value=@Add.Result)
   > &RigUnit_ModifyTransforms(
    ItemToModify=((Item=(Type=Bone,Name="coat_bck_base"),Transform=(Rotation=(X=0,Y=0,Z=-0.1736,W=0.9848),Translation=(X=0,Y=0,Z=0),Scale3D=(X=1,Y=1,Z=1))))
    WeightMinimum=0
    Weight=@Clamp.Result
   )
   Cast = &RigVMFunction_MathBoolToFloat(Value=RigVMVariableNode() {variable=IsSlidingKnee}.Value)
   n6 = &RigVMFunction_MathDoubleRemap(
    SourceMaximum=25
    SourceMinimum=40
    TargetMaximum=1
    TargetMinimum=0
    Value=&RigUnit_GetRelativeTransformForItem(Child=(Type=Bone,Name="pelvis"), Parent=(Type=Bone,Name="root"), bChildInitial=False, bParentInitial=False).RelativeTransform.Translation.Z
   )
   Multiply = &RigVMFunction_MathDoubleMul(B=&RigVMFunction_MathDoubleMul(A=@Remap.Result, B=@n5.Result).Result, A=@n6.Result)
   n7 = &RigVMFunction_MathDoubleMul(B=@Cast.Result, A=@Multiply.Result)
   > &RigUnit_ModifyTransforms(
    ItemToModify="((Item=(Type=Bone,Name=\"coat_l_base\"),Transform=(Rotation=(X=-0.1169,Y=0.4244,Z=-0.1739,W=0.8809),Translation=(X=-5,Y=0,Z=0),Scale3D=(X=1,Y=1,Z=1))),(Item=(Type=Bone,Name=\"coat_r_base\"),Transform=(Rotation=(X=0.1169,Y=-0.4244,Z=-0.1739,W=0.8809),Translatio ...(truncated)"
    WeightMinimum=0
    Weight=@n7.Result
   )
   > &RigUnit_ModifyTransforms(
    ItemToModify=((Item=(Type=Bone,Name="dyn_flur"),Transform=(Rotation=(X=0,Y=0,Z=0,W=1),Translation=(X=0,Y=-30,Z=2),Scale3D=(X=1,Y=1,Z=1))))
    WeightMinimum=0
    Weight=@n7.Result
   )
   n8 = &RigVMFunction_MathDoubleMul(B=&RigVMFunction_MathBoolToFloat(Value=RigVMVariableNode() {variable=IsSlidingBaseball}.Value).Result, A=@n6.Result)
   > &RigUnit_ModifyTransforms(
    ItemToModify="((Item=(Type=Bone,Name=\"coat_l_base\"),Transform=(Rotation=(X=-0.1294,Y=0.483,Z=-0.2999,W=0.8124),Translation=(X=0,Y=0,Z=0),Scale3D=(X=1,Y=1,Z=1))),(Item=(Type=Bone,Name=\"coat_r_base\"),Transform=(Rotation=(X=-0.1333,Y=0.0813,Z=-0.8999,W=0.4071),Translation= ...(truncated)"
    WeightMinimum=0
    Weight=@n8.Result
   )
   > &RigUnit_ModifyTransforms(
    ItemToModify=((Item=(Type=Bone,Name="dyn_flur"),Transform=(Rotation=(X=0,Y=0,Z=0,W=1),Translation=(X=0,Y=-25,Z=5),Scale3D=(X=1,Y=1,Z=1))))
    WeightMinimum=0
    Weight=@n8.Result
   )
   RigVMFunction_MathQuaternionToEuler = &RigVMFunction_MathQuaternionToEuler(
    RotationOrder=XYZ
    Value=&RigUnit_GetRelativeTransformForItem(Child=(Type=Bone,Name="spine_05"), Parent=(Type=Bone,Name="pelvis"), bChildInitial=False, bParentInitial=False).RelativeTransform.Rotation
   )
   > &RigUnit_ModifyTransforms(
    ItemToModify="((Item=(Type=Bone,Name=\"coat_fwd1_r_base\"),Transform=(Rotation=(X=-0.099,Y=-0.3696,Z=-0.2391,W=0.8924),Translation=(X=0,Y=0,Z=0),Scale3D=(X=1,Y=1,Z=1))),(Item=(Type=Bone,Name=\"coat_r_base\"),Transform=(Rotation=(X=-0.0076,Y=-0.0868,Z=-0.0868,W=0.9924),Trans ...(truncated)"
    WeightMinimum=0
    Weight=&RigVMFunction_MathDoubleMul(
     A=@n8.Result
     B=&RigVMFunction_MathDoubleRemap(SourceMaximum=45, SourceMinimum=0, TargetMaximum=1, TargetMinimum=0, Value=@RigVMFunction_MathQuaternionToEuler.Result.X).Result
    ).Result
   )
   > &RigUnit_ModifyTransforms(
    ItemToModify="((Item=(Type=Bone,Name=\"coat_fwd1_r_base\"),Transform=(Rotation=(X=-0.2549,Y=-0.0449,Z=0.1677,W=0.9513),Translation=(X=0,Y=-2,Z=3),Scale3D=(X=1,Y=1,Z=1))),(Item=(Type=Bone,Name=\"coat_r_base\"),Transform=(Rotation=(X=0.2308,Y=0.1687,Z=0.133,W=0.949),Translati ...(truncated)"
    WeightMinimum=0
    Weight=&RigVMFunction_MathDoubleMul(
     A=@n8.Result
     B=&RigVMFunction_MathDoubleRemap(SourceMaximum=-50, SourceMinimum=-10, TargetMaximum=1, TargetMinimum=0, Value=0).Result
    ).Result
   )
   RigVMFunction_MathBoolToFloat = &RigVMFunction_MathBoolToFloat(Value=RigVMVariableNode() {variable=IsParachuteOpenedAndCharUpRight}.Value)
   RigVMFunction_AlphaInterp = &RigVMFunction_AlphaInterp(
    InterpSpeedDecreasing=10
    InterpSpeedIncreasing=10
    Scale=1
    bClampResult=False
    bInterpResult=True
    Value=@RigVMFunction_MathBoolToFloat.Result
   )
   > &RigUnit_ParentConstraint(
    AdvancedSettings=(InterpolationType=Average,RotationOrderForFilter=XZY)
    Child=(Type=Bone,Name="coat_fwd1_l_base")
    Parents=((Item=(Type=Bone,Name="pelvis"),Weight=0.5),(Item=(Type=Bone,Name="thigh_l"),Weight=0.5))
    Weight=@RigVMFunction_AlphaInterp.Result
   )
   n9 = &RigVMFunction_AlphaInterp(
    InterpSpeedDecreasing=10
    InterpSpeedIncreasing=10
    Scale=1
    bClampResult=False
    bInterpResult=True
    Value=@RigVMFunction_MathBoolToFloat.Result
   )
   > &RigUnit_ParentConstraint(
    AdvancedSettings=(InterpolationType=Average,RotationOrderForFilter=XZY)
    Child=(Type=Bone,Name="coat_fwd1_r_base")
    Parents=((Item=(Type=Bone,Name="pelvis"),Weight=0.5),(Item=(Type=Bone,Name="thigh_l"),Weight=0.5))
    Weight=@n9.Result
   )
   n10 = &RigVMFunction_AlphaInterp(
    InterpSpeedDecreasing=10
    InterpSpeedIncreasing=10
    Scale=1
    bClampResult=False
    bInterpResult=True
    Value=@RigVMFunction_MathBoolToFloat.Result
   )
   > &RigUnit_ParentConstraint(
    AdvancedSettings=(InterpolationType=Average,RotationOrderForFilter=XZY)
    Child=(Type=Bone,Name="coat_bck_base")
    Parents=((Item=(Type=Bone,Name="pelvis"),Weight=1),(Item=(Type=Bone,Name="thigh_l"),Weight=1),(Item=(Type=Bone,Name="thigh_r"),Weight=1))
    Weight=@n10.Result
   )
   n11 = &RigVMFunction_AlphaInterp(
    InterpSpeedDecreasing=10
    InterpSpeedIncreasing=10
    Scale=1
    bClampResult=False
    bInterpResult=True
    Value=@RigVMFunction_MathBoolToFloat.Result
   )
   > &RigUnit_ParentConstraint(
    AdvancedSettings=(InterpolationType=Average,RotationOrderForFilter=XZY)
    Child=(Type=Bone,Name="coat_l_base")
    Parents=((Item=(Type=Bone,Name="pelvis"),Weight=0.5),(Item=(Type=Bone,Name="thigh_l"),Weight=1))
    Weight=@n11.Result
   )
   n12 = &RigVMFunction_AlphaInterp(
    InterpSpeedDecreasing=10
    InterpSpeedIncreasing=10
    Scale=1
    bClampResult=False
    bInterpResult=True
    Value=@RigVMFunction_MathBoolToFloat.Result
   )
   > &RigUnit_ParentConstraint(
    AdvancedSettings=(InterpolationType=Average,RotationOrderForFilter=XZY)
    Child=(Type=Bone,Name="coat_r_base")
    Parents=((Item=(Type=Bone,Name="pelvis"),Weight=0.5),(Item=(Type=Bone,Name="thigh_l"),Weight=1))
    Weight=@n12.Result
   )
   n13 = &RigVMFunction_AlphaInterp(
    InterpSpeedDecreasing=10
    InterpSpeedIncreasing=10
    Scale=1
    bClampResult=False
    bInterpResult=True
    Value=@RigVMFunction_MathBoolToFloat.Result
   )
   > &RigUnit_ModifyTransforms(
    ItemToModify=((Item=(Type=Bone,Name="coat_fwd1_l_base"),Transform=(Rotation=(X=0.2685,Y=0.0381,Z=-0.1893,W=0.9437),Translation=(X=0,Y=0,Z=0),Scale3D=(X=1,Y=1,Z=1))))
    WeightMinimum=0
    Weight=@n13.Result
   )
   n14 = &RigVMFunction_AlphaInterp(
    InterpSpeedDecreasing=10
    InterpSpeedIncreasing=10
    Scale=1
    bClampResult=False
    bInterpResult=True
    Value=@RigVMFunction_MathBoolToFloat.Result
   )
   > &RigUnit_ModifyTransforms(
    ItemToModify=((Item=(Type=Bone,Name="coat_fwd1_r_base"),Transform=(Rotation=(X=-0.2685,Y=-0.0381,Z=-0.1893,W=0.9437),Translation=(X=0,Y=0,Z=0),Scale3D=(X=1,Y=1,Z=1))))
    WeightMinimum=0
    Weight=@n14.Result
   )
   n15 = &RigVMFunction_AlphaInterp(
    InterpSpeedDecreasing=10
    InterpSpeedIncreasing=10
    Scale=1
    bClampResult=False
    bInterpResult=True
    Value=@RigVMFunction_MathBoolToFloat.Result
   )
   > &RigUnit_ModifyTransforms(
    ItemToModify="((Item=(Type=Bone,Name=\"coat_bck_base\"),Transform=(Rotation=(X=0,Y=0,Z=-0.0872,W=0.9962),Translation=(X=0,Y=0,Z=0),Scale3D=(X=1,Y=1,Z=1))),(Item=(Type=Bone,Name=\"coat_l_base\"),Transform=(Rotation=(X=0.1736,Y=0,Z=0,W=0.9848),Translation=(X=0,Y=0,Z=-3),Scale ...(truncated)"
    WeightMinimum=0
    Weight=@n15.Result
   )
   n16 = &RigUnit_GetRelativeTransformForItem(Child=(Type=Bone,Name="pelvis"), Parent=(Type=Bone,Name="root"), bChildInitial=False, bParentInitial=False)
   n17 = &RigVMFunction_MathDoubleMul(
    B=&RigVMFunction_MathBoolToFloat(Value=RigVMVariableNode() {variable=isSkydiveFloating}.Value).Result
    A=&RigVMFunction_MathDoubleRemap(
     SourceMaximum=65
     SourceMinimum=40
     TargetMaximum=1
     TargetMinimum=0
     Value=&RigVMFunction_MathQuaternionToEuler(RotationOrder=ZXY, Value=@n16.RelativeTransform.Rotation).Result.Z
    ).Result
   )
   > &RigUnit_ParentConstraint(
    AdvancedSettings=(InterpolationType=Average,RotationOrderForFilter=XZY)
    Child=(Type=Bone,Name="coat_fwd1_l_base")
    Parents=((Item=(Type=Bone,Name="thigh_twist_01_l"),Weight=1),(Item=(Type=Bone,Name="pelvis"),Weight=0))
    Weight=@n17.Result
   )
   > &RigUnit_ParentConstraint(
    AdvancedSettings=(InterpolationType=Average,RotationOrderForFilter=XZY)
    Child=(Type=Bone,Name="coat_fwd1_r_base")
    Parents=((Item=(Type=Bone,Name="thigh_twist_01_r"),Weight=1),(Item=(Type=Bone,Name="pelvis"),Weight=0))
    Weight=@n17.Result
   )
   > &RigUnit_ModifyTransforms(
    ItemToModify="((Item=(Type=Bone,Name=\"coat_fwd1_l_base\"),Transform=(Rotation=(X=0,Y=0,Z=0,W=1),Translation=(X=0,Y=-10,Z=-5),Scale3D=(X=1,Y=1,Z=1))),(Item=(Type=Bone,Name=\"coat_fwd1_r_base\"),Transform=(Rotation=(X=0,Y=0,Z=0,W=1),Translation=(X=0,Y=-10,Z=5),Scale3D=(X=1,Y ...(truncated)"
    WeightMinimum=0
    Weight=@n17.Result
   )
   &RigUnit_ParentConstraint(
    AdvancedSettings=(InterpolationType=Average,RotationOrderForFilter=XZY)
    Child=(Type=Control,Name="upperarm_twist_01_r_ctrl")
    Filter=(TranslationFilter=(bX=true,bY=true,bZ=true),RotationFilter=(bX=true,bY=true,bZ=true),ScaleFilter=(bX=True,bY=True,bZ=True))
    Parents=((Item=(Type=Bone,Name="upperarm_r"),Weight=0.5),(Item=(Type=Bone,Name="clavicle_r"),Weight=0.5))
    Weight=1
    ExecutePin=@RigVMFunction_Sequence.C
   )
   > &RigUnit_ParentConstraint(
    AdvancedSettings=(InterpolationType=Average,RotationOrderForFilter=XZY)
    Child=(Type=Control,Name="upperarm_twist_01_r_ctrl")
    Filter=(TranslationFilter=(bX=true,bY=true,bZ=true),RotationFilter=(bX=false,bY=false,bZ=false),ScaleFilter=(bX=True,bY=True,bZ=True))
    Parents=((Item=(Type=Bone,Name="upperarm_r"),Weight=0.5),(Item=(Type=Bone,Name="clavicle_r"),Weight=0.5),(Item=(Type=Bone,Name="deform_upperarm_up_r"),Weight=1))
    Weight=1
   )
   > &RigUnit_AimConstraintLocalSpaceOffset(
    Child=(Type=Control,Name="upperarm_twist_01_r_ctrl")
    Filter=(bX=false,bY=true,bZ=true)
    Parents=((Item=(Type=Control,Name="upperarm_r_ctrl"),Weight=1))
    Weight=1
    WorldUp=(Target=(X=0,Y=0,Z=1),Kind=Direction,Space=(Type=None,Name="None"))
   )
   > &RigUnit_ModifyTransforms(
    ItemToModify=((Item=(Type=Control,Name="upperarm_twist_01_r_ctrl"),Transform=(Rotation=(X=0,Y=0,Z=0,W=1),Translation=(X=0,Y=0,Z=0),Scale3D=(X=1,Y=1.1,Z=1))))
    Weight=1
    WeightMinimum=0
   )
   > n18 = &RigUnit_SphericalPoseReader(
    ActiveRegionScaleFactors=(PositiveWidth=1.2,NegativeWidth=1.5,PositiveHeight=1.5,NegativeHeight=1.5)
    ActiveRegionSize=0.11
    Debug=(bDrawDebug=false,bDraw2D=False,bDrawLocalAxes=False,DebugScale=25,DebugSegments=20,DebugThickness=0.25)
    DriverAxis=(X=-1,Y=0,Z=0)
    DriverItem=(Type=Bone,Name="upperarm_r")
    FalloffRegionScaleFactors=(PositiveWidth=0.5,NegativeWidth=1,PositiveHeight=1,NegativeHeight=1)
    FalloffSize=0.5
    FlipHeightScaling=False
    FlipWidthScaling=False
    OptionalParentItem=(Type=Bone,Name="None")
    RotationOffset=(X=-10,Y=130,Z=0)
   )
   n19 = RigVMVariableNode() {variable=IsParachuteOpenedAndCharUpRight}
   AlphaInterp = &RigVMFunction_AlphaInterp(
    InterpSpeedDecreasing=10
    InterpSpeedIncreasing=10
    Scale=1
    bClampResult=False
    Value=&RigVMFunction_MathDoubleMul(B=&RigVMFunction_MathBoolToFloat(Value=@n19.Value).Result, A=@n18.OutputParam).Result
   )
   > &RigUnit_ModifyTransforms(
    ItemToModify=((Item=(Type=Control,Name="upperarm_twist_01_r_ctrl"),Transform=(Rotation=(X=-0.1294,Y=-0.017,Z=0.1294,W=0.983),Translation=(X=-2,Y=0,Z=1),Scale3D=(X=1,Y=1,Z=1))))
    WeightMinimum=0
    Weight=@AlphaInterp.Result
   )
   > n20 = &RigUnit_SphericalPoseReader(
    ActiveRegionScaleFactors=(PositiveWidth=2,NegativeWidth=0.5,PositiveHeight=2,NegativeHeight=2)
    ActiveRegionSize=0.1
    Debug=(bDrawDebug=True,bDraw2D=False,bDrawLocalAxes=False,DebugScale=25,DebugSegments=20,DebugThickness=0.25)
    DriverAxis=(X=-1,Y=0,Z=0)
    DriverItem=(Type=Bone,Name="upperarm_r")
    FalloffRegionScaleFactors=(PositiveWidth=1,NegativeWidth=1,PositiveHeight=0.5,NegativeHeight=0.5)
    FalloffSize=0.3
    FlipHeightScaling=False
    FlipWidthScaling=False
    OptionalParentItem=(Type=Bone,Name="None")
    RotationOffset=(X=-172,Y=10,Z=0)
   )
   RigVMFunction_MathBoolNot = &RigVMFunction_MathBoolNot(Value=RigVMVariableNode() {variable=IsParachuteOpenedAndCharUpRight}.Value)
   RigVMFunction_MathBoolAnd = &RigVMFunction_MathBoolAnd(B=@RigVMFunction_MathBoolNot.Result, A=RigVMVariableNode() {variable=isParachuteOpen}.Value)
   n21 = &RigUnit_GetRelativeTransformForItem(Child=(Type=Bone,Name="pelvis"), Parent=(Type=Bone,Name="root"), bChildInitial=False, bParentInitial=False)
   Multiply_9 = RigVMAggregateNode(
    C=&RigVMFunction_MathBoolToFloat(Value=@RigVMFunction_MathBoolAnd.Result).Result
    B=&RigVMFunction_MathDoubleRemap(
     SourceMaximum=75
     SourceMinimum=50
     TargetMaximum=1
     TargetMinimum=0
     Value=&RigVMFunction_MathQuaternionToEuler(RotationOrder=XZY, Value=@n21.RelativeTransform.Rotation).Result.Z
    ).Result
    A=@n20.OutputParam
   )
   > &RigUnit_ModifyTransforms(
    ItemToModify=((Item=(Type=Control,Name="upperarm_twist_01_r_ctrl"),Transform=(Rotation=(X=0.13,Y=0.0864,Z=-0.0114,W=0.9877),Translation=(X=-2,Y=0,Z=-3),Scale3D=(X=1,Y=1,Z=1))))
    WeightMinimum=0
    Weight=@Multiply_9.Result
   )
   Get Delta Ref = RigVMFunctionReferenceNode(Item=(Type=Bone,Name="foot_l"), ExecuteContext=@RigVMFunction_Sequence.D)
   > &RigUnit_ModifyTransforms(
    ItemToModify=((Item=(Type=Bone,Name="def_foot_corr_bk_l"),Transform=(Rotation=(X=0.5,Y=0,Z=0,W=0.866),Translation=(X=-6,Y=-6,Z=4),Scale3D=(X=1,Y=1,Z=1))))
    WeightMinimum=0
    Weight=&RigVMFunction_MathDoubleRemap(SourceMaximum=70, SourceMinimum=0, TargetMaximum=1, TargetMinimum=0, Value=@"Get Delta Ref".Result.Rotation.Yaw).Result
   )
   > Get Delta Ref_1 = RigVMFunctionReferenceNode(Item=(Type=Bone,Name="foot_r"))
   > &RigUnit_ModifyTransforms(
    ItemToModify=((Item=(Type=Bone,Name="def_foot_corr_bk_r"),Transform=(Rotation=(X=0.5,Y=0,Z=0,W=0.866),Translation=(X=-6,Y=-6,Z=4),Scale3D=(X=1,Y=1,Z=1))))
    WeightMinimum=0
    Weight=&RigVMFunction_MathDoubleRemap(SourceMaximum=70, SourceMinimum=0, TargetMaximum=1, TargetMinimum=0, Value=@"Get Delta Ref_1".Result.Rotation.Yaw).Result
   )
  ]}
]
hierarchy [
 root {loc=<0 0 0>, rot=<0 0 0 1>}
 pelvis {parent=root, loc=<0 -0.7071 93.1191>, rot=<0.0127 0.707 -0.0127 -0.707>}
 spine_01 {parent=pelvis, loc=<6.7494 -0.6388 0>, rot=<0 0 0.034 -0.9994>}
 spine_02 {parent=spine_01, loc=<7.2518 -0.035 0>, rot=<0 0 0.0238 -0.9997>}
 spine_03 {parent=spine_02, loc=<6.0104 0 0>, rot=<0 0 0.0149 -0.9999>}
 spine_04 {parent=spine_03, loc=<8.1024 0 0>, rot=<0 0 -0.0132 -0.9999>}
 spine_05 {parent=spine_04, loc=<10.0146 0 0>, rot=<0 0 -0.0015 -1>}
 neck_01 {parent=spine_05, loc=<11.3138 0.1529 0>, rot=<0 0 -0.2003 -0.9797>}
 neck_02 {parent=neck_01, loc=<4.5627 0 0>, rot=<0 0 0.0504 -0.9987>}
 head {parent=neck_02, loc=<5.3914 0 0>, rot=<0 0 0.115 -0.9934>}
 attach_cape {parent=spine_05, loc=<9.017 -11.714 0>, rot=<0.0616 -0.7044 -0.0616 -0.7044>}
 cape {parent=attach_cape, type=Null, loc=<0 0 0>, rot=<0 0 0 1>}
 attach_backpack {parent=spine_05, loc=<-0.123 -13.473 0>, rot=<-0.0093 -0.707 0.0093 -0.707>}
 Backpack_BR {parent=attach_backpack, type=Null, loc=<0 0 0>, rot=<0 0 0 1>}
 clavicle_l {parent=spine_05, loc=<4.3575 -0.1466 -2.3223>, rot=<0.0465 -0.7289 -0.009 -0.6829>}
 upperarm_l {parent=clavicle_l, loc=<15.5111 0 0>, rot=<0.0226 -0.3234 -0.0088 -0.9459>}
 lowerarm_l {parent=upperarm_l, loc=<26.6514 0 0>, rot=<0.0391 -0.0099 -0.2453 -0.9686>}
 hand_l {parent=lowerarm_l, loc=<22.3491 0 0>, rot=<-0.7081 -0.0507 0.0975 -0.6975>}
 weapon_l {parent=hand_l, loc=<-0.1974 -3.4111 0.3811>, rot=<0 0 0 1>}
 thumb_01_l {parent=hand_l, loc=<1.3118 -0.3191 -2.459>, rot=<0.6546 -0.4042 0.0664 -0.6354>}
 thumb_02_l {parent=thumb_01_l, loc=<5.9336 -0.0448 0.0401>, rot=<-0.0037 0.0213 0.298 -0.9543>}
 thumb_03_l {parent=thumb_02_l, loc=<2.2223 0.3682 -0.0907>, rot=<0.0118 -0.013 0.0066 -0.9998>}
 index_metacarpal_l {parent=hand_l, loc=<1.6535 0.0183 -2.96>, rot=<0.0158 -0.0124 -0.0173 -0.9996>}
 index_01_l {parent=index_metacarpal_l, loc=<8.3008 0.0357 -0.0524>, rot=<-0.0044 0.0485 0.0924 -0.9945>}
 index_02_l {parent=index_01_l, loc=<3.3587 -0.3004 0.0452>, rot=<-0.002 0.0027 0.3225 -0.9466>}
 index_03_l {parent=index_02_l, loc=<2.5578 -0.123 -0.0252>, rot=<-0.002 0.0067 0.2558 -0.9667>}
 middle_metacarpal_l {parent=hand_l, loc=<2.1856 -0.1725 -0.5984>, rot=<0.006 0.0168 -0.0329 -0.9993>}
 middle_01_l {parent=middle_metacarpal_l, loc=<7.3275 -0.2447 0.0129>, rot=<-0.0208 0.0042 0.0858 -0.9961>}
 middle_02_l {parent=middle_01_l, loc=<4.0123 0.1401 0.1175>, rot=<-0.0218 0.0258 0.3512 -0.9357>}
 middle_03_l {parent=middle_02_l, loc=<2.6077 0.0534 0.1373>, rot=<0.0082 0.0188 0.1708 -0.9851>}
 ring_metacarpal_l {parent=hand_l, loc=<2.2074 -0.3625 1.3638>, rot=<-0.0284 0.0722 -0.0458 -0.9959>}
 ring_01_l {parent=ring_metacarpal_l, loc=<7.0295 -0.6197 -0.1446>, rot=<-0.0489 -0.0252 0.1892 -0.9804>}
 ring_02_l {parent=ring_01_l, loc=<3.6174 0.4647 -0.1027>, rot=<0 0 0.2806 -0.9598>}
 ring_03_l {parent=ring_02_l, loc=<2.4152 0.114 0.0199>, rot=<0 0 0.1189 -0.9929>}
 pinky_metacarpal_l {parent=hand_l, loc=<1.7616 -0.9753 3.3814>, rot=<-0.0397 0.0814 -0.069 -0.9935>}
 pinky_01_l {parent=pinky_metacarpal_l, loc=<7.1709 -0.4831 -0.1094>, rot=<-0.0654 -0.0035 0.2242 -0.9723>}
 pinky_02_l {parent=pinky_01_l, loc=<2.8274 0.0893 -0.1043>, rot=<0.0037 0.0011 0.2854 -0.9584>}
 pinky_03_l {parent=pinky_02_l, loc=<2.2046 0.0307 -0.0078>, rot=<-0.0002 0.0098 0.1445 -0.9895>}
 hand_attach_l {parent=hand_l, loc=<8.5 -3 0>, rot=<0.7071 0.7071 0 0>}
 deform_wrist_dn_base_l {parent=hand_l, loc=<0 -2.2342 0>, rot=<0 0 0 -1>}
 deform_wrist_dn_l {parent=deform_wrist_dn_base_l, loc=<0 0 0>, rot=<0 0 0 1>}
 deform_wrist_up_base_l {parent=hand_l, loc=<0 2.2682 0>, rot=<0 0 0 -1>}
 deform_wrist_up_l {parent=deform_wrist_up_base_l, loc=<0 0 0>, rot=<0 0 0 1>}
 lowerarm_twist_02_l {parent=lowerarm_l, loc=<9.6199 0 0>, rot=<0 0 0 -1>}
 lowerarm_twist_01_l {parent=lowerarm_l, loc=<17.8549 0 0>, rot=<0 0 0 -1>}
 deform_elbow_in_base_l {parent=lowerarm_l, loc=<0.2112 5.259 0.8274>, rot=<0 0 0 -1>}
 deform_elbow_in_l {parent=deform_elbow_in_base_l, loc=<0 0 0>, rot=<0 0 0 1>}
 deform_elbow_out_base_l {parent=lowerarm_l, loc=<0 -4 0>, rot=<0 0 0 -1>}
 deform_elbow_out_l {parent=deform_elbow_out_base_l, loc=<0 0 0>, rot=<0 0 0 1>}
 upperarm_twist_01_l {parent=upperarm_l, loc=<0.0041 0 0>, rot=<0 0 0 -1>}
 deform_upperarm_bk_base_l {parent=upperarm_twist_01_l, loc=<1.0813 -7.2193 -1.9453>, rot=<0 0 0 1>}
 deform_upperarm_bk_l {parent=deform_upperarm_bk_base_l, loc=<0 0 0>, rot=<0 0 0 1>}
 deform_upperarm_fr_base_l {parent=upperarm_twist_01_l, loc=<-0.3691 7.0506 -2.765>, rot=<0 0 0 1>}
 deform_upperarm_fr_l {parent=deform_upperarm_fr_base_l, loc=<0 0 0>, rot=<0 0 0 1>}
 deform_upperarm_up_base_l {parent=upperarm_twist_01_l, loc=<-3.2216 -1.1391 3.3311>, rot=<0 0 0 1>}
 deform_upperarm_up_l {parent=deform_upperarm_up_base_l, loc=<0 0 0>, rot=<0 0 0 1>}
 upperarm_twist_02_l {parent=upperarm_l, loc=<13.7941 0 0>, rot=<0 0 0 -1>}
 clavicle_r {parent=spine_05, loc=<4.3575 -0.1466 2.3223>, rot=<0.7289 0.0465 -0.6829 0.009>}
 upperarm_r {parent=clavicle_r, loc=<-15.5111 0 0>, rot=<0.0226 -0.3234 -0.0088 -0.9459>}
 lowerarm_r {parent=upperarm_r, loc=<-26.6514 0 0>, rot=<0.0391 -0.0099 -0.2453 -0.9686>}
 hand_r {parent=lowerarm_r, loc=<-22.3491 0 0>, rot=<-0.7081 -0.0507 0.0975 -0.6975>}
 weapon_r {parent=hand_r, loc=<0.1974 3.4111 -0.3811>, rot=<0 0 0 1>}
 thumb_01_r {parent=hand_r, loc=<-1.3118 0.3191 2.459>, rot=<0.6546 -0.4042 0.0664 -0.6354>}
 thumb_02_r {parent=thumb_01_r, loc=<-5.9336 0.0448 -0.0401>, rot=<-0.0037 0.0213 0.298 -0.9543>}
 thumb_03_r {parent=thumb_02_r, loc=<-2.2223 -0.3682 0.0907>, rot=<0.0118 -0.013 0.0066 -0.9998>}
 index_metacarpal_r {parent=hand_r, loc=<-1.6535 -0.0183 2.96>, rot=<0.0158 -0.0124 -0.0173 -0.9996>}
 index_01_r {parent=index_metacarpal_r, loc=<-8.3008 -0.0357 0.0524>, rot=<-0.0044 0.0485 0.0924 -0.9945>}
 index_02_r {parent=index_01_r, loc=<-3.3587 0.3004 -0.0452>, rot=<-0.002 0.0027 0.3225 -0.9466>}
 index_03_r {parent=index_02_r, loc=<-2.5578 0.123 0.0252>, rot=<-0.002 0.0067 0.2558 -0.9667>}
 middle_metacarpal_r {parent=hand_r, loc=<-2.1856 0.1725 0.5984>, rot=<0.006 0.0168 -0.0329 -0.9993>}
 middle_01_r {parent=middle_metacarpal_r, loc=<-7.3275 0.2447 -0.0129>, rot=<-0.0208 0.0042 0.0858 -0.9961>}
 middle_02_r {parent=middle_01_r, loc=<-4.0123 -0.1401 -0.1175>, rot=<-0.0218 0.0258 0.3512 -0.9357>}
 middle_03_r {parent=middle_02_r, loc=<-2.6077 -0.0534 -0.1373>, rot=<0.0082 0.0188 0.1708 -0.9851>}
 ring_metacarpal_r {parent=hand_r, loc=<-2.2074 0.3625 -1.3638>, rot=<-0.0284 0.0722 -0.0458 -0.9959>}
 ring_01_r {parent=ring_metacarpal_r, loc=<-7.0295 0.6197 0.1446>, rot=<-0.0489 -0.0252 0.1892 -0.9804>}
 ring_02_r {parent=ring_01_r, loc=<-3.6174 -0.4647 0.1027>, rot=<0 0 0.2806 -0.9598>}
 ring_03_r {parent=ring_02_r, loc=<-2.4152 -0.114 -0.0199>, rot=<0 0 0.1189 -0.9929>}
 pinky_metacarpal_r {parent=hand_r, loc=<-1.7616 0.9753 -3.3814>, rot=<-0.0397 0.0814 -0.069 -0.9935>}
 pinky_01_r {parent=pinky_metacarpal_r, loc=<-7.1709 0.4831 0.1094>, rot=<-0.0654 -0.0035 0.2242 -0.9723>}
 pinky_02_r {parent=pinky_01_r, loc=<-2.8274 -0.0893 0.1043>, rot=<0.0037 0.0011 0.2854 -0.9584>}
 pinky_03_r {parent=pinky_02_r, loc=<-2.2046 -0.0307 0.0078>, rot=<-0.0002 0.0098 0.1445 -0.9895>}
 hand_attach_r {parent=hand_r, loc=<-8.5001 2.9999 0>, rot=<0 0 -0.7071 -0.7071>}
 deform_wrist_dn_base_r {parent=hand_r, loc=<-0.0004 2.2345 -0.0001>, rot=<0 0 0 -1>}
 deform_wrist_dn_r {parent=deform_wrist_dn_base_r, loc=<0 0 0>, rot=<0 0 0 1>}
 deform_wrist_up_base_r {parent=hand_r, loc=<0.0003 -2.2684 0.0001>, rot=<0 0 0 -1>}
 deform_wrist_up_r {parent=deform_wrist_up_base_r, loc=<0 0 0>, rot=<0 0 0 1>}
 lowerarm_twist_02_r {parent=lowerarm_r, loc=<-9.6199 0 0>, rot=<0 0 0 -1>}
 lowerarm_twist_01_r {parent=lowerarm_r, loc=<-17.8549 0 0>, rot=<0 0 0 -1>}
 deform_elbow_in_base_r {parent=lowerarm_r, loc=<-0.2111 -5.2591 -0.8275>, rot=<0 0 0 -1>}
 deform_elbow_in_r {parent=deform_elbow_in_base_r, loc=<-0.0001 0 0.0002>, rot=<0 0 0 1>}
 deform_elbow_out_base_r {parent=lowerarm_r, loc=<0 4 -0.0001>, rot=<0 0 0 -1>}
 deform_elbow_out_r {parent=deform_elbow_out_base_r, loc=<0 0 0>, rot=<0 0 0 1>}
 upperarm_twist_01_r {parent=upperarm_r, loc=<-0.0041 0 0>, rot=<0 0 0 -1>}
 deform_upperarm_bk_base_r {parent=upperarm_twist_01_r, loc=<-1.0814 7.2192 1.9455>, rot=<0 0 0 1>}
 deform_upperarm_bk_r {parent=deform_upperarm_bk_base_r, loc=<0 0 0>, rot=<0 0 0 1>}
 deform_upperarm_fr_base_r {parent=upperarm_twist_01_r, loc=<0.3689 -7.0506 2.7652>, rot=<0 0 0 1>}
 deform_upperarm_fr_r {parent=deform_upperarm_fr_base_r, loc=<0 0 0>, rot=<0 0 0 1>}
 deform_upperarm_up_base_r {parent=upperarm_twist_01_r, loc=<3.2214 1.1391 -3.3308>, rot=<0 0 0 1>}
 deform_upperarm_up_r {parent=deform_upperarm_up_base_r, loc=<0 0 0>, rot=<0 0 0 -1>}
 upperarm_twist_01_r_ctrl {controlType=EULER_TRANSFORM, parent=upperarm_twist_01_r, type=Control, loc=<0 0 0>, rot=<0 0 0 1>}
 upperarm_twist_02_r {parent=upperarm_r, loc=<-13.7941 0 0>, rot=<0 0 0 -1>}
 shoulder_cor1 {parent=upperarm_r, loc=<-13.896 0.0735 -5.1268>, rot=<0 0 0 -1>}
 shoulder_cor2 {parent=upperarm_r, loc=<-7.0021 8.1469 0.3312>, rot=<0 0 0 1>}
 shoulder_cor3 {parent=upperarm_r, loc=<-6.792 -5.8771 -0.2157>, rot=<0 0 0 1>}
 shoulder_cor4 {parent=upperarm_r, loc=<4.2155 1.2693 -6.8117>, rot=<0 0 0 -1>}
 upperarm_r_ctrl {controlType=EULER_TRANSFORM, parent=upperarm_r, type=Control, loc=<0 0 0>, rot=<0 0 0 1>}
 dyn_zipper {parent=spine_05, loc=<8.0914 7.9196 -0.0348>, rot=<-0.0533 -0.9983 0.0109 0.0205>}
 deform_pec_aim_base_l {parent=spine_04, loc=<7.8732 12.6388 -2.4123>, rot=<0.0102 -0.4602 0.3979 -0.7936>}
 deform_pec_aim_l {parent=deform_pec_aim_base_l, loc=<0 0 0>, rot=<0 0 0 1>}
 deform_pec_base_l {parent=deform_pec_aim_l, loc=<10.766 0 0>, rot=<0 0 0 1>}
 deform_pec_l {parent=deform_pec_base_l, loc=<0.055 0 0>, rot=<0 0 0 1>}
 deform_pec_aim_base_r {parent=spine_04, loc=<7.8727 12.6388 2.4123>, rot=<-0.4602 -0.0102 0.7936 0.3979>}
 deform_pec_aim_r {parent=deform_pec_aim_base_r, loc=<0 0 0>, rot=<0 0 0 1>}
 deform_pec_base_r {parent=deform_pec_aim_r, loc=<-10.7662 -0.0006 -0.0007>, rot=<0 0 0 1>}
 deform_pec_r {parent=deform_pec_base_r, loc=<0 0 0>, rot=<0 0 0 1>}
 bag_up_l_02 {parent=spine_03, loc=<0.8943 8.5346 -10.192>, rot=<0.3236 -0.2279 -0.7188 -0.5716>}
 bag_up_l_01 {parent=spine_03, loc=<3.9289 3.2677 -15.3483>, rot=<-0.4847 0.2581 0.6754 0.4922>}
 bag_up_r_02 {parent=spine_03, loc=<0.8945 8.5346 10.192>, rot=<-0.2268 0.1208 -0.7863 -0.5618>}
 bag_up_r_01 {parent=spine_03, loc=<3.6804 3.9185 15.0487>, rot=<-0.4128 0.2717 -0.6807 -0.5409>}
 coat_fwd1_r_base {parent=spine_01, loc=<3.846 7.4707 8.8163>, rot=<0.0452 -0.2412 0.9691 0.0251>}
 coat_fwd1_r_01 {parent=coat_fwd1_r_base, loc=<0 0 0>, rot=<0 0 0 1>}
 coat_fwd1_r_02 {parent=coat_fwd1_r_01, loc=<23.0003 0 -0.0001>, rot=<-0.0642 0 0 -0.9979>}
 coat_fwd1_r_03 {parent=coat_fwd1_r_02, loc=<22.1823 0 0>, rot=<-0.0248 0 0 -0.9997>}
 stripe_fwd1_r_base {parent=coat_fwd1_r_02, loc=<1.8208 0.5052 3.6836>, rot=<0.2146 -0.0131 0.0077 -0.9766>}
 stripe_fwd1_r_01 {parent=stripe_fwd1_r_base, loc=<0 0 0>, rot=<-0.0492 0.0105 -0.0005 -0.9987>}
 stripe_fwd1_r_02 {parent=stripe_fwd1_r_01, loc=<15.1446 0.0001 -0.0001>, rot=<0 0 0 1>}
 bag_r {parent=coat_fwd1_r_01, loc=<14.3226 0.6595 6.3635>, rot=<0.0624 0.24 0.6689 -0.7008>}
 coat_fwd1_l_base {parent=spine_01, loc=<3.846 7.4707 -8.8163>, rot=<-0.0452 0.2412 0.9691 0.0251>}
 coat_fwd1_l_01 {parent=coat_fwd1_l_base, loc=<0 0 0>, rot=<0 0 0 1>}
 coat_fwd1_l_02 {parent=coat_fwd1_l_01, loc=<23.0003 0.0001 0.0001>, rot=<0.0642 0 0 -0.9979>}
 coat_fwd1_l_03 {parent=coat_fwd1_l_02, loc=<22.1823 0 0>, rot=<-0.0248 0 0 -0.9997>}
 stripe_fwd1_l_base {parent=coat_fwd1_l_02, loc=<1.8208 0.5052 -3.6835>, rot=<-0.2146 0.0131 0.0077 -0.9766>}
 stripe_fwd1_l_01 {parent=stripe_fwd1_l_base, loc=<0 0 0>, rot=<0.0492 -0.0105 -0.0005 -0.9987>}
 stripe_fwd1_l_02 {parent=stripe_fwd1_l_01, loc=<15.1446 0 0>, rot=<0 0 0 1>}
 bag_l {parent=coat_fwd1_l_01, loc=<14.3226 0.6593 -6.363>, rot=<0.0064 -0.2226 0.623 -0.7499>}
 coat_r_base {parent=spine_01, loc=<2.36 -1.6998 14.1639>, rot=<0.0903 -0.7152 0.6911 0.0524>}
 coat_r_01 {parent=coat_r_base, loc=<0 0 0>, rot=<0 0 0 1>}
 coat_r_02 {parent=coat_r_01, loc=<22.6051 0 0>, rot=<-0.0408 -0.026 -0.002 -0.9988>}
 coat_r_03 {parent=coat_r_02, loc=<21.3262 0 0>, rot=<0 0.0028 -0.0324 -0.9995>}
 stripe_r_base {parent=coat_r_02, loc=<-1.6911 -0.9805 -1.6607>, rot=<0.0442 0.013 0.01 -0.9989>}
 stripe_r_01 {parent=stripe_r_base, loc=<0 0 0>, rot=<0 0 0 1>}
 stripe_r_02 {parent=stripe_r_01, loc=<17 0 0>, rot=<0 0 0 1>}
 stripe_r_03 {parent=stripe_r_02, loc=<15 0.3689 0>, rot=<-0.021 -0.0005 -0.0218 -0.9995>}
 stripe_r_04 {parent=stripe_r_03, loc=<15.0143 0 0.3776>, rot=<-0.0384 0.0006 0.016 -0.9991>}
 coat_l_base {parent=spine_01, loc=<2.36 -1.6998 -14.1639>, rot=<0.0903 -0.7152 -0.6911 -0.0524>}
 coat_l_01 {parent=coat_l_base, loc=<0 0 0>, rot=<0 0 0 1>}
 coat_l_02 {parent=coat_l_01, loc=<22.6051 0 0>, rot=<0.0408 0.026 -0.002 -0.9988>}
 coat_l_03 {parent=coat_l_02, loc=<21.3262 0 0>, rot=<0 -0.0028 -0.0324 -0.9995>}
 stripe_l_base {parent=coat_l_02, loc=<-1.6911 -0.9805 1.6607>, rot=<-0.0442 -0.013 0.01 -0.9989>}
 stripe_l_01 {parent=stripe_l_base, loc=<0 0 0>, rot=<0 0 0 1>}
 stripe_l_02 {parent=stripe_l_01, loc=<17 0 0>, rot=<0 0 0 1>}
 stripe_l_03 {parent=stripe_l_02, loc=<15 0.3689 0>, rot=<0.021 0.0005 -0.0218 -0.9995>}
 stripe_l_04 {parent=stripe_l_03, loc=<15.0143 0 -0.3776>, rot=<0.0384 -0.0006 0.016 -0.9991>}
 coat_bck_base {parent=spine_01, loc=<2.419 -11.5495 0>, rot=<0.0416 -0.9991 0 0>}
 coat_bck_01 {parent=coat_bck_base, loc=<0 0 0>, rot=<0 0 0 1>}
 stripe_bck_m_l_base {parent=coat_bck_01, loc=<9.4056 -1.3518 6.5349>, rot=<0.0615 0.0335 0.0068 -0.9975>}
 stripe_bck_m_l_01 {parent=stripe_bck_m_l_base, loc=<0 0 0>, rot=<0 0 0 1>}
 stripe_bck_m_l_02 {parent=stripe_bck_m_l_01, loc=<10.8797 -0.098 0.0977>, rot=<0 0 0 1>}
 stripe_bck_m_l_03 {parent=stripe_bck_m_l_02, loc=<9.8243 0.1439 0.1089>, rot=<0 0 -0.0234 -0.9997>}
 stripe_bck_m_r_base {parent=coat_bck_01, loc=<9.4057 -1.3518 -6.5349>, rot=<-0.0615 -0.0335 0.0068 -0.9975>}
 stripe_bck_m_r_01 {parent=stripe_bck_m_r_base, loc=<0 0 0>, rot=<0 0 0 1>}
 stripe_bck_m_r_02 {parent=stripe_bck_m_r_01, loc=<10.8797 -0.0981 -0.0977>, rot=<0 0 0 1>}
 stripe_bck_m_r_03 {parent=stripe_bck_m_r_02, loc=<9.8243 0.144 -0.1089>, rot=<0 0 -0.0234 -0.9997>}
 coat_bck_02 {parent=coat_bck_01, loc=<22 0 0>, rot=<0 0 0 1>}
 coat_bck_03 {parent=coat_bck_02, loc=<21.5676 0 0>, rot=<0 0 0 1>}
 thigh_l {parent=pelvis, loc=<-2.5693 -0.0109 -7.7501>, rot=<0.0012 0.0572 0.0155 -0.9982>}
 calf_l {parent=thigh_l, loc=<-39.0484 0 0>, rot=<0.0006 -0.0161 -0.0391 -0.9991>}
 foot_l {parent=calf_l, loc=<-42.561 0.1077 0.4292>, rot=<0.0418 -0.0455 0.0343 -0.9975>}
 ball_l {parent=foot_l, loc=<-6.1599 13.8264 0.6806>, rot=<0.0001 -0.0001 -0.7063 -0.7079>}
 calf_twist_02_l {parent=calf_l, loc=<-18.8064 -0.0967 0.0397>, rot=<0.0237 0.0095 -0.0035 -0.9997>}
 calf_twist_01_l {parent=calf_l, loc=<-35.813 -0.0967 0.0397>, rot=<0.0237 0.0095 -0.0035 -0.9997>}
 def_foot_corr_bk_l {parent=calf_twist_01_l, loc=<-2.512 -4.8074 0.1962>, rot=<-0.0498 -0.7417 0.0123 -0.6688>}
 deform_knee_in_base_l {parent=calf_l, loc=<1.298 -5.2957 -0.0496>, rot=<0 0 0 -1>}
 deform_knee_in_l {parent=deform_knee_in_base_l, loc=<0 0 0>, rot=<0 0 0 1>}
 deform_knee_out_base_l {parent=calf_l, loc=<0 4 0>, rot=<0 0 0 -1>}
 deform_knee_out_l {parent=deform_knee_out_base_l, loc=<0 0 0>, rot=<0 0 0 1>}
 thigh_twist_01_l {parent=thigh_l, loc=<-21.6303 0 0>, rot=<0.0013 0 0 -1>}
 deform_groin_base_l {parent=thigh_l, loc=<-2.1908 8.1142 -1.7442>, rot=<0 0 0 -1>}
 deform_groin_l {parent=deform_groin_base_l, loc=<0 0 0>, rot=<0 0 0 1>}
 coat_fwd1_l_base_ctrl {controlType=EULER_TRANSFORM, parent=thigh_l, type=Control, loc=<0 0 0>, rot=<0 0 0 1>}
 thigh_r {parent=pelvis, loc=<-2.5693 -0.0109 7.7501>, rot=<0.0572 -0.0012 0.9982 0.0155>}
 calf_r {parent=thigh_r, loc=<39.0484 0 0>, rot=<0.0006 -0.0161 -0.0391 -0.9991>}
 foot_r {parent=calf_r, loc=<42.561 -0.1077 -0.4292>, rot=<0.0418 -0.0455 0.0343 -0.9975>}
 ball_r {parent=foot_r, loc=<6.1599 -13.8264 -0.6806>, rot=<0.0001 -0.0001 -0.7063 -0.7079>}
 calf_twist_02_r {parent=calf_r, loc=<18.8064 0.0967 -0.0397>, rot=<0.0237 0.0095 -0.0035 -0.9997>}
 calf_twist_01_r {parent=calf_r, loc=<35.813 0.0967 -0.0397>, rot=<0.0237 0.0095 -0.0035 -0.9997>}
 def_foot_corr_bk_r {parent=calf_twist_01_r, loc=<2.512 4.8074 -0.1962>, rot=<-0.6688 0.0123 0.7417 0.0498>}
 deform_knee_in_base_r {parent=calf_r, loc=<-1.298 5.2957 0.0496>, rot=<0 0 0 -1>}
 deform_knee_in_r {parent=deform_knee_in_base_r, loc=<0 0 0>, rot=<0 0 0 1>}
 deform_knee_out_base_r {parent=calf_r, loc=<0 -4 0>, rot=<0 0 0 -1>}
 deform_knee_out_r {parent=deform_knee_out_base_r, loc=<0 0 0>, rot=<0 0 0 1>}
 thigh_twist_01_r {parent=thigh_r, loc=<21.6303 0 0>, rot=<0.0013 0 0 -1>}
 deform_groin_base_r {parent=thigh_r, loc=<2.1908 -8.1142 1.7442>, rot=<0 0 0 -1>}
 deform_groin_r {parent=deform_groin_base_r, loc=<0 0 0>, rot=<0 0 0 1>}
 thigh_r_ctrl {controlType=EULER_TRANSFORM, parent=thigh_r, type=Control, loc=<0 0 0>, rot=<0 0 0 1>}
 deform_glute_base_l {parent=pelvis, loc=<-8.8064 -8.8222 -8.4287>, rot=<0.0275 0.0501 -0.4472 -0.8926>}
 deform_glute_l {parent=deform_glute_base_l, loc=<0 0 0>, rot=<0 0 0 1>}
 deform_glute_base_r {parent=pelvis, loc=<-8.8062 -8.8222 8.4287>, rot=<0.0501 -0.0275 0.8926 -0.4472>}
 deform_glute_r {parent=deform_glute_base_r, loc=<0 0 0>, rot=<0 0 0 1>}
 dyn_flur {parent=root, loc=<0 0 0>, rot=<0 0 0 1>}
]
variables [
 IsParachuteOpenedAndCharUpRight {type=bool, default=False}
 IsSlidingBaseball {type=bool, default=False}
 IsSlidingKnee {type=bool, default=False}
 Leg ROM_Both Thighs Back {type=double, default=0}
 Leg ROM_L Foot on Thigh R {type=double, default=0}
 Leg ROM_L Knee {type=double, default=0}
 Leg ROM_L Thigh Raised Fwd Side Inbetween {type=float, default=0}
 Leg ROM_L Thigh Raised Rear {type=float, default=0}
 Leg ROM_L Thigh Raised Side {type=float, default=0}
 Leg ROM_R Knee {type=double, default=0}
 Leg ROM_R Thigh Raised Fwd Side Inbetween {type=float, default=0}
 Leg ROM_R Thigh Raised Rear {type=float, default=0}
 Leg ROM_R Thigh Raised Side {type=float, default=0}
 LegForward_L {type=float, default=0}
 LegForward_R {type=float, default=0}
 isDBNO {type=bool, default=False}
 isParachuteOpen {type=bool, default=False}
 isSkydiveFloating {type=bool, default=False}
]
