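# Anim Blueprint for the RankedPhantom male medium body, adding secondary motion to its coat, bags, stripes and zipper.
# Two coat RigidBody sims on a cached Control Rig pose blend by parachute state, using the skydive physics asset when open.
# Weighted constraints attach bags and stripes, then a stripes Control Rig, Look Ats and per-bone AnimDynamics refine them.
# The rig's knee-slide and baseball-slide inputs receive each other's flags.
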AnimBlueprint /BRCosmetics/Characters/Player/Male/Medium/Bodies/M_MED_RankedPhantom/Meshes/M_MED_RankedPhantom_AnimBP
target_skeleton=&/M_MED_RankedPhantom_Skeleton
graphs [
 AnimGraph {nodes [
   n1 = Look At - Bone: stripe_bck_m_l_03(Alpha=1) {
    bUseLookUpAxis=true
    boneToModify {boneName=stripe_bck_m_l_03}
    lookAtLocation=<0 -2.1674 7.8>
    lookAtTarget {boneReference {boneName=coat_bck_03}}
    lookAt_Axis {axis=<1 0>}
    lookUp_Axis {axis {x=0, z=1}}
   }
   Look At - Bone: stripe_bck_m_r_03(ComponentPose=@n1, Alpha=1) {
    bUseLookUpAxis=true
    boneToModify {boneName=stripe_bck_m_r_03}
    lookAtLocation=<0 -1.65 -7.8746>
    lookAtTarget {boneReference {boneName=coat_bck_03}}
    lookAt_Axis {axis=<1 0>}
    lookUp_Axis {axis {x=0, z=1}}
   }
   n2 = RigidBody!(ComponentPose=Local To Component(LocalPose=Use cached pose 'Control RigDef'()), Alpha, OverrideWorldGravity:Vector="0,0,-980", ExternalForce) -> (Pose) {
    bClampLinearTranslationLimitToRefPose=true
    bFreezeIncomingPoseOnStart=true
    bOverrideWorldGravity=true
    overridePhysicsAsset=&/M_MED_RankedPhantom_Physics_coat3
    simulationSpace=BaseBoneSpace
    baseBoneRef {boneName=spine_01}
    overrideWorldGravity {z=-980}
    simSpaceSettings {dampingAlpha=0, maxAngularAcceleration=30, maxAngularVelocity=6, maxLinearAcceleration=120, maxLinearVelocity=8, worldAlpha=1}
   }
   n3 = Component To Local(ComponentPose=@n2)
   n4 = RigidBody(ComponentPose=Local To Component(LocalPose=Use cached pose 'Control RigDef'()), OverrideWorldGravity="0,0,-980") {
    bClampLinearTranslationLimitToRefPose=true
    bFreezeIncomingPoseOnStart=true
    bOverrideWorldGravity=true
    overridePhysicsAsset=&/M_MED_RankedPhantom_Physics_CoatSkyDive
    simulationSpace=BaseBoneSpace
    baseBoneRef {boneName=spine_01}
    overrideWorldGravity {z=-980}
    simSpaceSettings {dampingAlpha=0, maxAngularAcceleration=30, maxAngularVelocity=6, maxLinearAcceleration=120, maxLinearVelocity=8, worldAlpha=1}
   }
   n5 = Component To Local(ComponentPose=@n4)
   n6 = Get bIsParachuteOpened()
   n7 = "Constraint - bag_l (PPPPPPPP)"!(
    ComponentPose=Local To Component(LocalPose=Blend Poses by bool(BlendPose_0=@n5, BlendPose_1=@n3, bActiveValue=@n6))
    Alpha=1
    ConstraintWeights_0=0.0596
    ConstraintWeights_1:float=0.0138
    ConstraintWeights_2:float=0.0059
    ConstraintWeights_3:float=0.0001
    ConstraintWeights_4:float=0.2205
    ConstraintWeights_5:float=0.224
    ConstraintWeights_6:float=0.3836
    ConstraintWeights_7:float=0.0926
   ) -> (Pose) {
    boneToModify {boneName=bag_l}
    constraintSetup [
     {targetBone {boneName=pelvis}}
     {targetBone {boneName=spine_01}}
     {targetBone {boneName=spine_02}}
     {targetBone {boneName=coat_bck_01}}
     {targetBone {boneName=coat_l_01}}
     {targetBone {boneName=coat_l_02}}
     {targetBone {boneName=coat_fwd1_l_01}}
     {targetBone {boneName=coat_fwd1_l_02}}
    ]
    constraintWeights [1,1,1,1,1,1,1,1]
   }
   n8 = "Constraint - bag_r (PPPPPPPP)"!(
    ComponentPose=@n7
    Alpha=1
    ConstraintWeights_0=0.0832
    ConstraintWeights_1:float=0.0206
    ConstraintWeights_2:float=0.0171
    ConstraintWeights_3:float=0.0116
    ConstraintWeights_4:float=0.2865
    ConstraintWeights_5:float=0.1389
    ConstraintWeights_6:float=0.3861
    ConstraintWeights_7:float=0.0559
   ) -> (Pose) {
    boneToModify {boneName=bag_r}
    constraintSetup [
     {targetBone {boneName=pelvis}}
     {targetBone {boneName=spine_01}}
     {targetBone {boneName=spine_02}}
     {targetBone {boneName=spine_03}}
     {targetBone {boneName=coat_r_01}}
     {targetBone {boneName=coat_r_02}}
     {targetBone {boneName=coat_fwd1_r_01}}
     {targetBone {boneName=coat_fwd1_r_02}}
    ]
    constraintWeights [1,1,1,1,1,1,1,1]
   }
   n9 = "Constraint - bag_up_l_02 (PPPPPP)"!(
    ComponentPose=@n8
    Alpha=1
    ConstraintWeights_0=0.0944
    ConstraintWeights_1:float=0.0193
    ConstraintWeights_2:float=0.2542
    ConstraintWeights_3:float=0.6139
    ConstraintWeights_4:float=0.0166
    ConstraintWeights_5:float=0.0017
   ) -> (Pose) {
    boneToModify {boneName=bag_up_l_02}
    constraintSetup [
     {targetBone {boneName=pelvis}}
     {targetBone {boneName=spine_01}}
     {targetBone {boneName=spine_02}}
     {targetBone {boneName=spine_03}}
     {targetBone {boneName=spine_04}}
     {targetBone {boneName=coat_l_01}}
    ]
    constraintWeights [1,1,1,1,1,1]
   }
   n10 = "Constraint - bag_up_l_01 (PPPPPP)"!(
    ComponentPose=@n9
    Alpha=1
    ConstraintWeights_0=0
    ConstraintWeights_1:float=0.0008
    ConstraintWeights_2:float=0.9927
    ConstraintWeights_3:float=0
    ConstraintWeights_4:float=0
    ConstraintWeights_5:float=0.0065
   ) -> (Pose) {
    boneToModify {boneName=bag_up_l_01}
    constraintSetup [
     {targetBone {boneName=pelvis}}
     {targetBone {boneName=spine_02}}
     {targetBone {boneName=spine_03}}
     {targetBone {boneName=deform_upperarm_fr_l}}
     {targetBone {boneName=deform_pec_aim_l}}
     {targetBone {boneName=deform_pec_l}}
    ]
    constraintWeights [1,1,1,1,1,1]
   }
   n11 = "Constraint - bag_up_r_01 (PPPPPPPP)"!(
    ComponentPose=@n10
    Alpha=1
    ConstraintWeights_0=0.0415
    ConstraintWeights_1:float=0
    ConstraintWeights_2:float=0.0711
    ConstraintWeights_3:float=0.859
    ConstraintWeights_4:float=0.005
    ConstraintWeights_5:float=0.0107
    ConstraintWeights_6:float=0.0127
    ConstraintWeights_7:float=0
   ) -> (Pose) {
    boneToModify {boneName=bag_up_r_01}
    constraintSetup [
     {targetBone {boneName=pelvis}}
     {targetBone {boneName=spine_01}}
     {targetBone {boneName=spine_02}}
     {targetBone {boneName=spine_03}}
     {targetBone {boneName=deform_upperarm_fr_r}}
     {targetBone {boneName=deform_pec_aim_r}}
     {targetBone {boneName=deform_pec_r}}
     {targetBone {boneName=coat_r_01}}
    ]
    constraintWeights [1,1,1,1,1,1,1,1]
   }
   n12 = "Constraint - bag_up_r_02 (PPPPPPPP)"!(
    ComponentPose=@n11
    Alpha=1
    ConstraintWeights_0=0.0947
    ConstraintWeights_1:float=0.0199
    ConstraintWeights_2:float=0.2536
    ConstraintWeights_3:float=0.6128
    ConstraintWeights_4:float=0.0173
    ConstraintWeights_5:float=0.0001
    ConstraintWeights_6:float=0.0017
    ConstraintWeights_7:float=0.0001
   ) -> (Pose) {
    boneToModify {boneName=bag_up_r_02}
    constraintSetup [
     {targetBone {boneName=pelvis}}
     {targetBone {boneName=spine_01}}
     {targetBone {boneName=spine_02}}
     {targetBone {boneName=spine_03}}
     {targetBone {boneName=spine_04}}
     {targetBone {boneName=deform_pec_aim_base_r}}
     {targetBone {boneName=coat_r_01}}
     {targetBone {boneName=coat_fwd1_r_01}}
    ]
    constraintWeights [1,1,1,1,1,1,1,1]
   }
   n13 = "Constraint - stripe_fwd1_l_base (PPPPPPP)"!(
    ComponentPose=@n12
    Alpha=1
    ConstraintWeights_0=0.004
    ConstraintWeights_1:float=0.0572
    ConstraintWeights_2:float=0.2503
    ConstraintWeights_3:float=0.0023
    ConstraintWeights_4:float=0.3538
    ConstraintWeights_5:float=0.3271
    ConstraintWeights_6:float=0.0053
   ) -> (Pose) {
    boneToModify {boneName=stripe_fwd1_l_base}
    constraintSetup [
     {targetBone {boneName=pelvis}}
     {targetBone {boneName=coat_l_01}}
     {targetBone {boneName=coat_l_02}}
     {targetBone {boneName=coat_l_03}}
     {targetBone {boneName=coat_fwd1_l_01}}
     {targetBone {boneName=coat_fwd1_l_02}}
     {targetBone {boneName=coat_fwd1_l_03}}
    ]
    constraintWeights [1,1,1,1,1,1,1]
   }
   n14 = "Constraint - stripe_fwd1_l_01 (PPPPPPP)"!(
    ComponentPose=@n13
    Alpha=1
    ConstraintWeights_0=0.0001
    ConstraintWeights_1:float=0.0207
    ConstraintWeights_2:float=0.1342
    ConstraintWeights_3:float=0.0207
    ConstraintWeights_4:float=0.2713
    ConstraintWeights_5:float=0.5297
    ConstraintWeights_6:float=0.0234
   ) -> (Pose) {
    boneToModify {boneName=stripe_fwd1_l_01}
    constraintSetup [
     {targetBone {boneName=pelvis}}
     {targetBone {boneName=coat_l_01}}
     {targetBone {boneName=coat_l_02}}
     {targetBone {boneName=coat_l_03}}
     {targetBone {boneName=coat_fwd1_l_01}}
     {targetBone {boneName=coat_fwd1_l_02}}
     {targetBone {boneName=coat_fwd1_l_03}}
    ]
    constraintWeights [1,1,1,1,1,1,1]
   }
   n15 = "Constraint - stripe_fwd1_r_02 (PPPPPP)"!(
    ComponentPose=@n14
    Alpha=1
    ConstraintWeights_0=0.0001
    ConstraintWeights_1:float=0.026
    ConstraintWeights_2:float=0.0764
    ConstraintWeights_3:float=0.0076
    ConstraintWeights_4:float=0.3869
    ConstraintWeights_5:float=0.5031
   ) -> (Pose) {
    boneToModify {boneName=stripe_fwd1_r_02}
    constraintSetup [
     {targetBone {boneName=coat_l_01}}
     {targetBone {boneName=coat_l_02}}
     {targetBone {boneName=coat_l_03}}
     {targetBone {boneName=coat_fwd1_l_01}}
     {targetBone {boneName=coat_fwd1_l_02}}
     {targetBone {boneName=coat_fwd1_l_03}}
    ]
    constraintWeights [1,1,1,1,1,1]
   }
   n16 = "Constraint - stripe_l_01 (PPPPPPPP)"!(
    ComponentPose=@n15
    Alpha=1
    ConstraintWeights_0=0
    ConstraintWeights_1:float=0.0008
    ConstraintWeights_2:float=0.1577
    ConstraintWeights_3:float=0.5398
    ConstraintWeights_4:float=0.0669
    ConstraintWeights_5:float=0.0536
    ConstraintWeights_6:float=0.1708
    ConstraintWeights_7:float=0.0102
   ) -> (Pose) {
    boneToModify {boneName=stripe_l_01}
    constraintSetup [
     {targetBone {boneName=pelvis}}
     {targetBone {boneName=coat_bck_01}}
     {targetBone {boneName=coat_l_01}}
     {targetBone {boneName=coat_l_02}}
     {targetBone {boneName=coat_l_03}}
     {targetBone {boneName=coat_fwd1_l_01}}
     {targetBone {boneName=coat_fwd1_l_02}}
     {targetBone {boneName=coat_fwd1_l_03}}
    ]
    constraintWeights [1,1,1,1,1,1,1,1]
   }
   n17 = "Constraint - stripe_l_02 (PPPPPPP)"!(
    ComponentPose=@n16
    Alpha=1
    ConstraintWeights_0=0
    ConstraintWeights_1:float=0.0018
    ConstraintWeights_2:float=0.5013
    ConstraintWeights_3:float=0.2199
    ConstraintWeights_4:float=0.0049
    ConstraintWeights_5:float=0.1705
    ConstraintWeights_6:float=0.1016
   ) -> (Pose) {
    boneToModify {boneName=stripe_l_02}
    constraintSetup [
     {targetBone {boneName=index_02_l}}
     {targetBone {boneName=coat_l_01}}
     {targetBone {boneName=coat_l_02}}
     {targetBone {boneName=coat_l_03}}
     {targetBone {boneName=coat_fwd1_l_01}}
     {targetBone {boneName=coat_fwd1_l_02}}
     {targetBone {boneName=coat_fwd1_l_03}}
    ]
    constraintWeights [1,1,1,1,1,1,1]
   }
   n18 = "Constraint - stripe_l_03 (PPPPPPP)"!(
    ComponentPose=@n17
    Alpha=1
    ConstraintWeights_0=0.0004
    ConstraintWeights_1:float=0.0004
    ConstraintWeights_2:float=0.0468
    ConstraintWeights_3:float=0.651
    ConstraintWeights_4:float=0.0004
    ConstraintWeights_5:float=0.0407
    ConstraintWeights_6:float=0.2603
   ) -> (Pose) {
    boneToModify {boneName=stripe_l_03}
    constraintSetup [
     {targetBone {boneName=ring_02_l}}
     {targetBone {boneName=ring_03_l}}
     {targetBone {boneName=coat_l_02}}
     {targetBone {boneName=coat_l_03}}
     {targetBone {boneName=coat_fwd1_l_01}}
     {targetBone {boneName=coat_fwd1_l_02}}
     {targetBone {boneName=coat_fwd1_l_03}}
    ]
    constraintWeights [1,1,1,1,1,1,1]
   }
   n19 = "Constraint - stripe_bck_m_l_base (PPPPPPPP)"!(
    ComponentPose=@n18
    Alpha=1
    ConstraintWeights_0=0.0567
    ConstraintWeights_1:float=0.0005
    ConstraintWeights_2:float=0.0014
    ConstraintWeights_3:float=0.0017
    ConstraintWeights_4:float=0.8273
    ConstraintWeights_5:float=0.0059
    ConstraintWeights_6:float=0.1003
    ConstraintWeights_7:float=0.0062
   ) -> (Pose) {
    boneToModify {boneName=stripe_bck_m_l_base}
    constraintSetup [
     {targetBone {boneName=pelvis}}
     {targetBone {boneName=spine_01}}
     {targetBone {boneName=spine_02}}
     {targetBone {boneName=spine_03}}
     {targetBone {boneName=coat_bck_01}}
     {targetBone {boneName=coat_bck_02}}
     {targetBone {boneName=coat_l_01}}
     {targetBone {boneName=coat_l_02}}
    ]
    constraintWeights [1,1,1,1,1,1,1,1]
   }
   n20 = "Constraint - stripe_bck_m_l_02 (PPPPPPPPP)"!(
    ComponentPose=@n19
    Alpha=1
    ConstraintWeights_0=0.0021
    ConstraintWeights_1:float=0
    ConstraintWeights_2:float=0.6303
    ConstraintWeights_3:float=0.2391
    ConstraintWeights_4:float=0.0016
    ConstraintWeights_5:float=0.0602
    ConstraintWeights_6:float=0.0666
    ConstraintWeights_7:float=0
    ConstraintWeights_8:float=0
   ) -> (Pose) {
    boneToModify {boneName=stripe_bck_m_l_02}
    constraintSetup [
     {targetBone {boneName=pelvis}}
     {targetBone {boneName=spine_01}}
     {targetBone {boneName=coat_bck_01}}
     {targetBone {boneName=coat_bck_02}}
     {targetBone {boneName=coat_bck_03}}
     {targetBone {boneName=coat_l_01}}
     {targetBone {boneName=coat_l_02}}
     {targetBone {boneName=coat_l_03}}
     {targetBone {boneName=coat_fwd1_l_01}}
    ]
    constraintWeights [1,1,1,1,1,1,1,1,1]
   }
   n21 = "Constraint - stripe_bck_m_l_03 (PPPPPPPP)"!(
    ComponentPose=@n20
    Alpha=1
    ConstraintWeights_0=0
    ConstraintWeights_1:float=0.0547
    ConstraintWeights_2:float=0.6693
    ConstraintWeights_3:float=0.0871
    ConstraintWeights_4:float=0.0002
    ConstraintWeights_5:float=0.1655
    ConstraintWeights_6:float=0.0232
    ConstraintWeights_7:float=0
   ) -> (Pose) {
    boneToModify {boneName=stripe_bck_m_l_03}
    constraintSetup [
     {targetBone {boneName=pelvis}}
     {targetBone {boneName=coat_bck_01}}
     {targetBone {boneName=coat_bck_02}}
     {targetBone {boneName=coat_bck_03}}
     {targetBone {boneName=coat_l_01}}
     {targetBone {boneName=coat_l_02}}
     {targetBone {boneName=coat_l_03}}
     {targetBone {boneName=coat_fwd1_l_01}}
    ]
    constraintWeights [1,1,1,1,1,1,1,1]
   }
   n22 = "Constraint - stripe_bck_m_r_base (PPPPPPPP)"!(
    ComponentPose=@n21
    Alpha=1
    ConstraintWeights_0=0.0537
    ConstraintWeights_1:float=0.0004
    ConstraintWeights_2:float=0.0012
    ConstraintWeights_3:float=0.0014
    ConstraintWeights_4:float=0.8333
    ConstraintWeights_5:float=0.0064
    ConstraintWeights_6:float=0.0973
    ConstraintWeights_7:float=0.0064
   ) -> (Pose) {
    boneToModify {boneName=stripe_bck_m_r_base}
    constraintSetup [
     {targetBone {boneName=pelvis}}
     {targetBone {boneName=spine_01}}
     {targetBone {boneName=spine_02}}
     {targetBone {boneName=spine_03}}
     {targetBone {boneName=coat_bck_01}}
     {targetBone {boneName=coat_bck_02}}
     {targetBone {boneName=coat_r_01}}
     {targetBone {boneName=coat_r_02}}
    ]
    constraintWeights [1,1,1,1,1,1,1,1]
   }
   n23 = "Constraint - stripe_bck_m_r_02 (PPPPPPPP)"!(
    ComponentPose=@n22
    Alpha=1
    ConstraintWeights_0=0.0019
    ConstraintWeights_1:float=0.5981
    ConstraintWeights_2:float=0.2718
    ConstraintWeights_3:float=0.002
    ConstraintWeights_4:float=0.0572
    ConstraintWeights_5:float=0.0688
    ConstraintWeights_6:float=0.0001
    ConstraintWeights_7:float=0.0001
   ) -> (Pose) {
    boneToModify {boneName=stripe_bck_m_r_02}
    constraintSetup [
     {targetBone {boneName=pelvis}}
     {targetBone {boneName=coat_bck_01}}
     {targetBone {boneName=coat_bck_02}}
     {targetBone {boneName=coat_bck_03}}
     {targetBone {boneName=coat_r_01}}
     {targetBone {boneName=coat_r_02}}
     {targetBone {boneName=coat_r_03}}
     {targetBone {boneName=coat_fwd1_r_01}}
    ]
    constraintWeights [1,1,1,1,1,1,1,1]
   }
   n24 = "Constraint - stripe_bck_m_r_03 (PPPPPPPP)"!(
    ComponentPose=@n23
    Alpha=1
    ConstraintWeights_0=0
    ConstraintWeights_1:float=0.0547
    ConstraintWeights_2:float=0.6718
    ConstraintWeights_3:float=0.0866
    ConstraintWeights_4:float=0.0002
    ConstraintWeights_5:float=0.1637
    ConstraintWeights_6:float=0.0229
    ConstraintWeights_7:float=0
   ) -> (Pose) {
    boneToModify {boneName=stripe_bck_m_r_03}
    constraintSetup [
     {targetBone {boneName=pelvis}}
     {targetBone {boneName=coat_bck_01}}
     {targetBone {boneName=coat_bck_02}}
     {targetBone {boneName=coat_bck_03}}
     {targetBone {boneName=coat_r_01}}
     {targetBone {boneName=coat_r_02}}
     {targetBone {boneName=coat_r_03}}
     {targetBone {boneName=coat_fwd1_r_01}}
    ]
    constraintWeights [1,1,1,1,1,1,1,1]
   }
   n25 = "Constraint - stripe_r_01 (PPPPPPPP)"!(
    ComponentPose=@n24
    Alpha=1
    ConstraintWeights_0=0
    ConstraintWeights_1:float=0
    ConstraintWeights_2:float=0.25
    ConstraintWeights_3:float=0.5
    ConstraintWeights_4:float=0
    ConstraintWeights_5:float=0
    ConstraintWeights_6:float=0.25
    ConstraintWeights_7:float=0
   ) -> (Pose) {
    boneToModify {boneName=stripe_r_01}
    constraintSetup [
     {targetBone {boneName=pelvis}}
     {targetBone {boneName=coat_bck_01}}
     {targetBone {boneName=coat_r_01}}
     {targetBone {boneName=coat_r_02}}
     {targetBone {boneName=coat_r_03}}
     {targetBone {boneName=coat_fwd1_r_01}}
     {targetBone {boneName=coat_fwd1_r_02}}
     {targetBone {boneName=coat_fwd1_r_03}}
    ]
    constraintWeights [1,1,1,1,1,1,1,1]
   }
   n26 = "Constraint - stripe_r_02 (PPPPPPP)"!(
    ComponentPose=@n25
    Alpha=1
    ConstraintWeights_0=0
    ConstraintWeights_1:float=0
    ConstraintWeights_2:float=0.6
    ConstraintWeights_3:float=0
    ConstraintWeights_4:float=0
    ConstraintWeights_5:float=0.4
    ConstraintWeights_6:float=0
   ) -> (Pose) {
    boneToModify {boneName=stripe_r_02}
    constraintSetup [
     {targetBone {boneName=index_03_r}}
     {targetBone {boneName=coat_r_01}}
     {targetBone {boneName=coat_r_02}}
     {targetBone {boneName=coat_r_03}}
     {targetBone {boneName=coat_fwd1_r_01}}
     {targetBone {boneName=coat_fwd1_r_02}}
     {targetBone {boneName=coat_fwd1_r_03}}
    ]
    constraintWeights [1,1,1,1,1,1,1]
   }
   n27 = "Constraint - stripe_r_03 (PPPPPP)"!(
    ComponentPose=@n26
    Alpha=1
    ConstraintWeights_0=0
    ConstraintWeights_1:float=0
    ConstraintWeights_2:float=0.75
    ConstraintWeights_3:float=0
    ConstraintWeights_4:float=0
    ConstraintWeights_5:float=0.25
   ) -> (Pose) {
    boneToModify {boneName=stripe_r_03}
    constraintSetup [
     {targetBone {boneName=middle_02_r}}
     {targetBone {boneName=coat_r_02}}
     {targetBone {boneName=coat_r_03}}
     {targetBone {boneName=coat_fwd1_r_01}}
     {targetBone {boneName=coat_fwd1_r_02}}
     {targetBone {boneName=coat_fwd1_r_03}}
    ]
    constraintWeights [1,1,1,1,1,1]
   }
   n28 = "Constraint - stripe_fwd1_r_base (PPPPPPP)"!(
    ComponentPose=@n27
    Alpha=1
    ConstraintWeights_0=0.0041
    ConstraintWeights_1:float=0.0577
    ConstraintWeights_2:float=0.2515
    ConstraintWeights_3:float=0.0023
    ConstraintWeights_4:float=0.3537
    ConstraintWeights_5:float=0.3254
    ConstraintWeights_6:float=0.0052
   ) -> (Pose) {
    boneToModify {boneName=stripe_fwd1_r_base}
    constraintSetup [
     {targetBone {boneName=pelvis}}
     {targetBone {boneName=coat_r_01}}
     {targetBone {boneName=coat_r_02}}
     {targetBone {boneName=coat_r_03}}
     {targetBone {boneName=coat_fwd1_r_01}}
     {targetBone {boneName=coat_fwd1_r_02}}
     {targetBone {boneName=coat_fwd1_r_03}}
    ]
    constraintWeights [1,1,1,1,1,1,1]
   }
   n29 = "Constraint - stripe_fwd1_r_01 (PPPPPPP)"!(
    ComponentPose=@n28
    Alpha=1
    ConstraintWeights_0=0.0001
    ConstraintWeights_1:float=0.0212
    ConstraintWeights_2:float=0.136
    ConstraintWeights_3:float=0.0206
    ConstraintWeights_4:float=0.272
    ConstraintWeights_5:float=0.5269
    ConstraintWeights_6:float=0.0231
   ) -> (Pose) {
    boneToModify {boneName=stripe_fwd1_r_01}
    constraintSetup [
     {targetBone {boneName=pelvis}}
     {targetBone {boneName=coat_r_01}}
     {targetBone {boneName=coat_r_02}}
     {targetBone {boneName=coat_r_03}}
     {targetBone {boneName=coat_fwd1_r_01}}
     {targetBone {boneName=coat_fwd1_r_02}}
     {targetBone {boneName=coat_fwd1_r_03}}
    ]
    constraintWeights [1,1,1,1,1,1,1]
   }
   n30 = "Constraint - stripe_fwd1_r_02 (PPPPPP)"(
    ComponentPose=@n29
    Alpha=1
    ConstraintWeights_0=0.0011
    ConstraintWeights_1=0.0456
    ConstraintWeights_2=0.0654
    ConstraintWeights_3=0.0245
    ConstraintWeights_4=0.5734
    ConstraintWeights_5=0.2901
   ) {
    boneToModify {boneName=stripe_fwd1_r_02}
    constraintSetup [
     {targetBone {boneName=coat_r_01}}
     {targetBone {boneName=coat_r_02}}
     {targetBone {boneName=coat_r_03}}
     {targetBone {boneName=coat_fwd1_r_01}}
     {targetBone {boneName=coat_fwd1_r_02}}
     {targetBone {boneName=coat_fwd1_r_03}}
    ]
    constraintWeights [1,1,1,1,1,1]
   }
   n31 = Control Rig!(
    Source=Component To Local(ComponentPose=@n30)
    Alpha
    isKneeSlide:bool=Get bIsBaseballSliding()
    isBaseBallSide:bool=Get bIsKneeSliding()
    isParachuteOpened:bool=Get bIsParachuteOpened()
   ) -> (Pose) {controlRigAssetReference {blueprintRigClass=&/M_MED_RankedPhantom__Stripes.M_MED_RankedPhantom__Stripes_C}}
   n32 = "Constraint - dyn_zipper (PPPPPP)"!(
    ComponentPose=Local To Component(LocalPose=@n31)
    Alpha=1
    ConstraintWeights_0=0.0466
    ConstraintWeights_1:float=0.4617
    ConstraintWeights_2:float=0.2458
    ConstraintWeights_3:float=0.0751
    ConstraintWeights_4:float=0.0854
    ConstraintWeights_5:float=0.0854
   ) -> (Pose) {
    boneToModify {boneName=dyn_zipper}
    constraintSetup [
     {targetBone {boneName=spine_04}}
     {targetBone {boneName=spine_05}}
     {targetBone {boneName=neck_01}}
     {targetBone {boneName=neck_02}}
     {targetBone {boneName=clavicle_l}}
     {targetBone {boneName=clavicle_r}}
    ]
    constraintWeights [1,1,1,1,1,1]
   }
   n33 = Anim Dynamics - Bone: dyn_zipper(ComponentPose=@n32, Alpha=1) {
    angularDampingOverride=0.999
    angularSpringConstant=600
    bAngularSpring=true
    bOverrideAngularDamping=true
    bUseGravityOverride=true
    simulationSpace=RootRelative
    boundBone {boneName=dyn_zipper}
    componentAppliedLinearAccClamp=<3000 3000 4000>
    componentLinearAccScale=<0.3 0.3 0.5>
    componentLinearVelScale=<0.3 0.3 0.7>
    gravityOverride {z=-980}
    physicsBodyDefinitions [{
      boundBone {boneName=dyn_zipper}
      boxExtents=<10 10 10>
      constraintSetup {angularLimitsMax=<5 10 75>, angularLimitsMin=<-5 -10 0>, angularTarget=<0 0 0>}
      localJointOffset=<0.7 0 0>
     }]
    simSpaceSettings {maxAngularAcceleration=350, maxAngularVelocity=30, simSpaceAngularAlpha=0.2}
   }
   n34 = "Transform (Modify) Bone - Bone: stripe_l_01"(ComponentPose=@n33, Alpha=1, Translation="0,-1,0", Rotation="0,0,0", Scale="1,1,1") {translationMode=BMM_Additive, translationSpace=BCS_BoneSpace, boneToModify {boneName=stripe_l_01}, translation {y=-1}}
   n35 = "Transform (Modify) Bone - Bone: stripe_r_01"(ComponentPose=@n34, Alpha=1, Translation="0,-1,0", Rotation="0,0,0", Scale="1,1,1") {translationMode=BMM_Additive, translationSpace=BCS_BoneSpace, boneToModify {boneName=stripe_r_01}, translation {y=-1}}
   n36 = Look At - Bone: stripe_l_base(ComponentPose=@n35, Alpha=1) {
    boneToModify {boneName=stripe_l_base}
    lookAtLocation=<0.013 -0.5618 2.4512>
    lookAtTarget {boneReference {boneName=coat_l_03}}
    lookAt_Axis {axis=<1 0>}
    lookUp_Axis {axis {x=0, z=1}}
   }
   n37 = Look At - Bone: stripe_r_base(ComponentPose=@n36, Alpha=1) {
    boneToModify {boneName=stripe_r_base}
    lookAtLocation=<0 -1.3611 0>
    lookAtTarget {boneReference {boneName=coat_r_03}}
    lookAt_Axis {axis=<1 0>}
    lookUp_Axis {axis {x=0, z=1}}
   }
   n38 = Look At - Bone: stripe_fwd1_l_base(ComponentPose=@n37, Alpha=0.5) {
    boneToModify {boneName=stripe_fwd1_l_base}
    lookAtLocation=<0 -0.0948 -4.0747>
    lookAtTarget {boneReference {boneName=coat_fwd1_l_03}}
    lookAt_Axis {axis=<1 0>}
    lookUp_Axis {axis {x=0, z=1}}
   }
   n39 = Look At - Bone: stripe_fwd1_r_base(ComponentPose=@n38, Alpha=0.5) {
    boneToModify {boneName=stripe_fwd1_r_base}
    lookAtLocation=<0 0.2262 4.3523>
    lookAtTarget {boneReference {boneName=coat_fwd1_r_03}}
    lookAt_Axis {axis=<1 0>}
    lookUp_Axis {axis {x=0, z=1}}
   }
   n40 = Look At - Bone: stripe_bck_m_l_01(ComponentPose=@n39, Alpha=1) {
    bUseLookUpAxis=true
    boneToModify {boneName=stripe_bck_m_l_01}
    lookAtLocation=<0 -2.0403 7.8>
    lookAtTarget {boneReference {boneName=coat_bck_02}}
    lookAt_Axis {axis=<1 0>}
    lookUp_Axis {axis {x=0, z=1}}
   }
   n41 = Look At - Bone: stripe_bck_m_r_01(ComponentPose=@n40, Alpha=1) {
    bUseLookUpAxis=true
    boneToModify {boneName=stripe_bck_m_r_01}
    lookAtLocation=<0 -2.0559 -7.8746>
    lookAtTarget {boneReference {boneName=coat_bck_02}}
    lookAt_Axis {axis=<1 0>}
    lookUp_Axis {axis {x=0, z=1}}
   }
   n42 = Anim Dynamics - Bone: stripe_bck_m_l_01(ComponentPose=@n41, Alpha=1, ExternalForce="0,0,0") {
    angularDampingOverride=0.99
    angularSpringConstant=300
    bAngularSpring=true
    bOverrideAngularDamping=true
    bUseGravityOverride=true
    simulationSpace=RootRelative
    boundBone {boneName=stripe_bck_m_l_01}
    componentAppliedLinearAccClamp=<3000 3000 5000>
    componentLinearAccScale=<0.3 0.3 0.8>
    componentLinearVelScale=<0.3 0.3 1>
    gravityOverride {z=-980}
    physicsBodyDefinitions [{
      boundBone {boneName=stripe_bck_m_l_01}
      boxExtents=<11 11 11>
      constraintSetup {angularLimitsMax=<0 0 0>, angularLimitsMin=<0 0 -35>, angularTarget=<0 0 0>}
      localJointOffset=<0.5 0 0>
     }]
    simSpaceSettings {maxAngularAcceleration=350, maxAngularVelocity=30, simSpaceAngularAlpha=0.3}
   }
   n43 = Anim Dynamics - Bone: stripe_bck_m_l_02(ComponentPose=@n42, Alpha=1, ExternalForce="0,0,0") {
    angularDampingOverride=0.99
    angularSpringConstant=300
    bAngularSpring=true
    bOverrideAngularDamping=true
    bUseGravityOverride=true
    simulationSpace=RootRelative
    boundBone {boneName=stripe_bck_m_l_02}
    componentAppliedLinearAccClamp=<3000 3000 5000>
    componentLinearAccScale=<0.3 0.3 0.8>
    componentLinearVelScale=<0.3 0.3 1>
    gravityOverride {z=-980}
    physicsBodyDefinitions [{
      boundBone {boneName=stripe_bck_m_l_02}
      boxExtents=<11.5 11.5 11.5>
      constraintSetup {angularLimitsMax=<0 0 0>, angularLimitsMin=<0 0 -30>, angularTarget=<0 0 0>}
      localJointOffset=<0.35 0 0>
     }]
    simSpaceSettings {maxAngularAcceleration=350, maxAngularVelocity=30, simSpaceAngularAlpha=0.3}
   }
   n44 = Anim Dynamics - Bone: stripe_bck_m_l_03(ComponentPose=@n43, Alpha=1, ExternalForce="0,0,0") {
    angularDampingOverride=0.98
    angularSpringConstant=300
    bAngularSpring=true
    bOverrideAngularDamping=true
    bUseGravityOverride=true
    simulationSpace=RootRelative
    boundBone {boneName=stripe_bck_m_l_03}
    componentAppliedLinearAccClamp=<3000 3000 5000>
    componentLinearAccScale=<0.3 0.3 0.8>
    componentLinearVelScale=<0.3 0.3 1>
    gravityOverride {z=-980}
    physicsBodyDefinitions [{
      boundBone {boneName=stripe_bck_m_l_03}
      boxExtents=<10 10 10>
      constraintSetup {angularLimitsMax=<0 3 0>, angularLimitsMin=<0 -3 -30>, angularTarget=<0 0 0>}
      localJointOffset=<0.7 0 0>
     }]
    simSpaceSettings {maxAngularAcceleration=350, maxAngularVelocity=30, simSpaceAngularAlpha=0.3}
   }
   n45 = Anim Dynamics - Bone: stripe_bck_m_r_01(ComponentPose=@n44, Alpha=1, ExternalForce="0,0,0") {
    angularDampingOverride=0.989
    angularSpringConstant=300
    bAngularSpring=true
    bOverrideAngularDamping=true
    bUseGravityOverride=true
    simulationSpace=RootRelative
    boundBone {boneName=stripe_bck_m_r_01}
    componentAppliedLinearAccClamp=<3000 3000 4500>
    componentLinearAccScale=<0.3 0.3 0.6>
    componentLinearVelScale=<0.3 0.3 0.8>
    gravityOverride {z=-980}
    physicsBodyDefinitions [{
      boundBone {boneName=stripe_bck_m_r_01}
      boxExtents=<12 12 12>
      constraintSetup {angularLimitsMax=<0 0 0>, angularLimitsMin=<0 0 -20>, angularTarget=<0 0 0>}
      localJointOffset=<0.7 0 0>
     }]
    simSpaceSettings {maxAngularAcceleration=350, maxAngularVelocity=30, simSpaceAngularAlpha=0.25}
   }
   n46 = Anim Dynamics - Bone: stripe_bck_m_r_02(ComponentPose=@n45, Alpha=1, ExternalForce="0,0,0") {
    angularDampingOverride=0.99
    angularSpringConstant=300
    bAngularSpring=true
    bOverrideAngularDamping=true
    bUseGravityOverride=true
    simulationSpace=RootRelative
    boundBone {boneName=stripe_bck_m_r_02}
    componentAppliedLinearAccClamp=<3000 3000 4500>
    componentLinearAccScale=<0.3 0.3 0.6>
    componentLinearVelScale=<0.3 0.3 0.8>
    gravityOverride {z=-980}
    physicsBodyDefinitions [{
      boundBone {boneName=stripe_bck_m_r_02}
      boxExtents=<12 12 12>
      constraintSetup {angularLimitsMax=<0 0 0>, angularLimitsMin=<0 0 -25>, angularTarget=<0 0 0>}
      localJointOffset=<0.5 0 0>
     }]
    simSpaceSettings {maxAngularAcceleration=350, maxAngularVelocity=30, simSpaceAngularAlpha=0.25}
   }
   n47 = Anim Dynamics - Bone: stripe_bck_m_r_03(ComponentPose=@n46, Alpha=1, ExternalForce="0,0,0") {
    angularDampingOverride=0.98
    angularSpringConstant=300
    bAngularSpring=true
    bOverrideAngularDamping=true
    bUseGravityOverride=true
    simulationSpace=RootRelative
    boundBone {boneName=stripe_bck_m_r_03}
    componentAppliedLinearAccClamp=<3000 3000 4500>
    componentLinearAccScale=<0.3 0.3 0.6>
    componentLinearVelScale=<0.3 0.3 0.8>
    gravityOverride {z=-980}
    physicsBodyDefinitions [{
      boundBone {boneName=stripe_bck_m_r_03}
      boxExtents=<11 11 11>
      constraintSetup {angularLimitsMax=<0 3 0>, angularLimitsMin=<0 -3 -30>, angularTarget=<0 0 0>}
      localJointOffset=<0.7 0 0>
     }]
    simSpaceSettings {maxAngularAcceleration=350, maxAngularVelocity=30, simSpaceAngularAlpha=0.25}
   }
   n48 = Anim Dynamics - Bone: stripe_fwd1_l_01(ComponentPose=@n47, Alpha=1, ExternalForce="0,0,0") {
    angularDampingOverride=0.99
    angularSpringConstant=300
    bAngularSpring=true
    bOverrideAngularDamping=true
    bUseGravityOverride=true
    simulationSpace=RootRelative
    boundBone {boneName=stripe_fwd1_l_01}
    componentAppliedLinearAccClamp=<3000 3000 5000>
    componentLinearAccScale=<0.3 0.3 0.8>
    componentLinearVelScale=<0.3 0.3 1>
    gravityOverride {z=-980}
    physicsBodyDefinitions [{
      boundBone {boneName=stripe_fwd1_l_01}
      boxExtents=<11 11 11>
      constraintSetup {angularLimitsMax=<0 3 0>, angularLimitsMin=<0 -3 -45>, angularTarget=<0 0 0>}
      localJointOffset=<0.6 0 0>
     }]
    simSpaceSettings {maxAngularAcceleration=350, maxAngularVelocity=30, simSpaceAngularAlpha=0.3}
   }
   n49 = Anim Dynamics - Bone: stripe_fwd1_l_02(ComponentPose=@n48, Alpha=1, ExternalForce="0,0,0") {
    angularDampingOverride=0.99
    angularSpringConstant=300
    bAngularSpring=true
    bOverrideAngularDamping=true
    bUseGravityOverride=true
    simulationSpace=RootRelative
    boundBone {boneName=stripe_fwd1_l_02}
    componentAppliedLinearAccClamp=<3000 3000 5000>
    componentLinearAccScale=<0.3 0.3 0.8>
    componentLinearVelScale=<0.3 0.3 1>
    gravityOverride {z=-980}
    physicsBodyDefinitions [{
      boundBone {boneName=stripe_fwd1_l_02}
      boxExtents=<11.5 11.5 11.5>
      constraintSetup {angularLimitsMax=<0 3 0>, angularLimitsMin=<0 -3 -60>, angularTarget=<0 0 0>}
      localJointOffset=<0.5 0 0>
     }]
    simSpaceSettings {maxAngularAcceleration=350, maxAngularVelocity=30, simSpaceAngularAlpha=0.3}
   }
   n50 = Anim Dynamics - Bone: stripe_fwd1_r_01(ComponentPose=@n49, Alpha=1, ExternalForce="0,0,0") {
    angularDampingOverride=0.99
    angularSpringConstant=300
    bAngularSpring=true
    bOverrideAngularDamping=true
    bUseGravityOverride=true
    simulationSpace=RootRelative
    boundBone {boneName=stripe_fwd1_r_01}
    componentAppliedLinearAccClamp=<3000 3000 4500>
    componentLinearAccScale=<0.3 0.3 0.7>
    componentLinearVelScale=<0.3 0.3 0.9>
    gravityOverride {z=-980}
    physicsBodyDefinitions [{
      boundBone {boneName=stripe_fwd1_r_01}
      boxExtents=<11 11 11>
      constraintSetup {angularLimitsMax=<0 3 0>, angularLimitsMin=<0 -3 -45>, angularTarget=<0 0 0>}
      localJointOffset=<0.5 0 0>
     }]
    simSpaceSettings {maxAngularAcceleration=350, maxAngularVelocity=30, simSpaceAngularAlpha=0.2}
   }
   n51 = Anim Dynamics - Bone: stripe_fwd1_r_02(ComponentPose=@n50, Alpha=1, ExternalForce="0,0,0") {
    angularDampingOverride=0.99
    angularSpringConstant=300
    bAngularSpring=true
    bOverrideAngularDamping=true
    bUseGravityOverride=true
    simulationSpace=RootRelative
    boundBone {boneName=stripe_fwd1_r_02}
    componentAppliedLinearAccClamp=<3000 3000 4500>
    componentLinearAccScale=<0.3 0.3 0.7>
    componentLinearVelScale=<0.3 0.3 0.9>
    gravityOverride {z=-980}
    physicsBodyDefinitions [{
      boundBone {boneName=stripe_fwd1_r_02}
      boxExtents=<10.5 10.5 10.5>
      constraintSetup {angularLimitsMax=<0 3 0>, angularLimitsMin=<0 -3 -60>, angularTarget=<0 0 0>}
      localJointOffset=<0.4 0 0>
     }]
    simSpaceSettings {maxAngularAcceleration=350, maxAngularVelocity=30, simSpaceAngularAlpha=0.2}
   }
   n52 = Anim Dynamics - Bone: stripe_l_01(ComponentPose=@n51, Alpha=1, ExternalForce="0,0,0") {
    angularDampingOverride=0.99
    angularSpringConstant=300
    bAngularSpring=true
    bOverrideAngularDamping=true
    bUseGravityOverride=true
    simulationSpace=RootRelative
    boundBone {boneName=stripe_l_01}
    componentAppliedLinearAccClamp=<3000 3000 5000>
    componentLinearAccScale=<0.3 0.3 0.8>
    componentLinearVelScale=<0.3 0.3 1>
    gravityOverride {z=-980}
    physicsBodyDefinitions [{
      boundBone {boneName=stripe_l_01}
      boxExtents=<11.5 11.5 11.5>
      constraintSetup {angularLimitsMax=<0 3 0>, angularLimitsMin=<0 -3 -45>, angularTarget=<0 0 0>}
      localJointOffset=<0.5 0 0>
     }]
    simSpaceSettings {maxAngularAcceleration=350, maxAngularVelocity=30, simSpaceAngularAlpha=0.3}
   }
   n53 = Anim Dynamics - Bone: stripe_l_02(ComponentPose=@n52, Alpha=1, ExternalForce="0,0,0") {
    angularDampingOverride=0.99
    angularSpringConstant=300
    bAngularSpring=true
    bOverrideAngularDamping=true
    bUseGravityOverride=true
    simulationSpace=RootRelative
    boundBone {boneName=stripe_l_02}
    componentAppliedLinearAccClamp=<3000 3000 5000>
    componentLinearAccScale=<0.3 0.3 0.8>
    componentLinearVelScale=<0.3 0.3 1>
    gravityOverride {z=-980}
    physicsBodyDefinitions [{
      boundBone {boneName=stripe_l_02}
      boxExtents=<11 11 11>
      constraintSetup {angularLimitsMax=<0 3 0>, angularLimitsMin=<0 -3 -45>, angularTarget=<0 0 0>}
      localJointOffset=<0.5 0 0>
     }]
    simSpaceSettings {maxAngularAcceleration=350, maxAngularVelocity=30, simSpaceAngularAlpha=0.3}
   }
   n54 = Anim Dynamics - Bone: stripe_l_03(ComponentPose=@n53, Alpha=1, ExternalForce="0,0,0") {
    angularDampingOverride=0.99
    angularSpringConstant=300
    bAngularSpring=true
    bOverrideAngularDamping=true
    bUseGravityOverride=true
    simulationSpace=RootRelative
    boundBone {boneName=stripe_l_03}
    componentAppliedLinearAccClamp=<3000 3000 5000>
    componentLinearAccScale=<0.3 0.3 0.8>
    componentLinearVelScale=<0.3 0.3 1>
    gravityOverride {z=-980}
    physicsBodyDefinitions [{
      collisionType=CustomSphere
      sphereCollisionRadius=3
      boundBone {boneName=stripe_l_03}
      boxExtents=<11 11 11>
      constraintSetup {angularLimitsMax=<0 3 15>, angularLimitsMin=<0 -3 -45>, angularTarget=<0 0 0>}
      localJointOffset=<0.7 0 0>
     }]
    planarLimits [{drivingBone {boneName=root}, planeTransform {location=<0 0 0>, rotation=<rot 0 0 0>, scale=<2 2 2>}}]
    simSpaceSettings {maxAngularAcceleration=350, maxAngularVelocity=30, simSpaceAngularAlpha=0.3}
   }
   n55 = Anim Dynamics - Bone: stripe_l_04(ComponentPose=@n54, Alpha=1, ExternalForce="0,0,0") {
    angularDampingOverride=0.99
    angularSpringConstant=300
    bAngularSpring=true
    bOverrideAngularDamping=true
    bUseGravityOverride=true
    simulationSpace=RootRelative
    boundBone {boneName=stripe_l_04}
    componentAppliedLinearAccClamp=<3000 3000 5000>
    componentLinearAccScale=<0.3 0.3 0.8>
    componentLinearVelScale=<0.3 0.3 1>
    gravityOverride {z=-980}
    physicsBodyDefinitions [{
      collisionType=CustomSphere
      sphereCollisionRadius=3
      boundBone {boneName=stripe_l_04}
      boxExtents=<10 10 10>
      constraintSetup {angularLimitsMax=<0 3 20>, angularLimitsMin=<0 -3 -20>, angularTarget=<0 0 0>}
      localJointOffset=<0.6 0 0>
     }]
    planarLimits [{drivingBone {boneName=root}, planeTransform {location=<0 0 0>, rotation=<rot 0 0 0>, scale=<2 2 2>}}]
    simSpaceSettings {maxAngularAcceleration=350, maxAngularVelocity=30, simSpaceAngularAlpha=0.3}
   }
   n56 = Anim Dynamics - Bone: stripe_r_01(ComponentPose=@n55, Alpha=1, ExternalForce="0,0,0") {
    angularDampingOverride=0.99
    angularSpringConstant=300
    bAngularSpring=true
    bOverrideAngularDamping=true
    bUseGravityOverride=true
    simulationSpace=RootRelative
    boundBone {boneName=stripe_r_01}
    componentAppliedLinearAccClamp=<3000 3000 5000>
    componentLinearAccScale=<0.3 0.3 0.8>
    componentLinearVelScale=<0.3 0.3 1>
    gravityOverride {z=-980}
    physicsBodyDefinitions [{
      boundBone {boneName=stripe_r_01}
      boxExtents=<12 12 12>
      constraintSetup {angularLimitsMax=<0 3 0>, angularLimitsMin=<0 -3 -45>, angularTarget=<0 0 0>}
      localJointOffset=<0.7 0 0>
     }]
    simSpaceSettings {maxAngularAcceleration=350, maxAngularVelocity=30, simSpaceAngularAlpha=0.3}
   }
   n57 = Anim Dynamics - Bone: stripe_r_02(ComponentPose=@n56, Alpha=1, ExternalForce="0,0,0") {
    angularDampingOverride=0.99
    angularSpringConstant=300
    bAngularSpring=true
    bOverrideAngularDamping=true
    bUseGravityOverride=true
    simulationSpace=RootRelative
    boundBone {boneName=stripe_r_02}
    componentAppliedLinearAccClamp=<3000 3000 4500>
    componentLinearAccScale=<0.3 0.3 0.75>
    componentLinearVelScale=<0.3 0.3 0.9>
    gravityOverride {z=-980}
    physicsBodyDefinitions [{
      boundBone {boneName=stripe_r_02}
      boxExtents=<11 11 11>
      constraintSetup {angularLimitsMax=<0 3 0>, angularLimitsMin=<0 -3 -45>, angularTarget=<0 0 0>}
      localJointOffset=<0.7 0 0>
     }]
    simSpaceSettings {maxAngularAcceleration=350, maxAngularVelocity=30, simSpaceAngularAlpha=0.25}
   }
   n58 = Anim Dynamics - Bone: stripe_r_03(ComponentPose=@n57, Alpha=1, ExternalForce="0,0,0") {
    angularDampingOverride=0.99
    angularSpringConstant=300
    bAngularSpring=true
    bOverrideAngularDamping=true
    bUseGravityOverride=true
    simulationSpace=RootRelative
    boundBone {boneName=stripe_r_03}
    componentAppliedLinearAccClamp=<3000 3000 4500>
    componentLinearAccScale=<0.3 0.3 0.75>
    componentLinearVelScale=<0.3 0.3 0.9>
    gravityOverride {z=-980}
    physicsBodyDefinitions [{
      collisionType=CustomSphere
      sphereCollisionRadius=3
      boundBone {boneName=stripe_r_03}
      boxExtents=<10.5 10.5 10.5>
      constraintSetup {angularLimitsMax=<0 3 15>, angularLimitsMin=<0 -3 -15>, angularTarget=<0 0 0>}
      localJointOffset=<0.8 0 0>
     }]
    planarLimits [{drivingBone {boneName=root}, planeTransform {location=<0 0 0>, rotation=<rot 0 0 0>, scale=<2 2 2>}}]
    simSpaceSettings {maxAngularAcceleration=350, maxAngularVelocity=30, simSpaceAngularAlpha=0.25}
   }
   n59 = Anim Dynamics - Bone: stripe_r_04(ComponentPose=@n58, Alpha=1, ExternalForce="0,0,0") {
    angularDampingOverride=0.99
    angularSpringConstant=300
    bAngularSpring=true
    bOverrideAngularDamping=true
    bUseGravityOverride=true
    simulationSpace=RootRelative
    boundBone {boneName=stripe_r_04}
    componentAppliedLinearAccClamp=<3000 3000 4500>
    componentLinearAccScale=<0.3 0.3 0.75>
    componentLinearVelScale=<0.3 0.3 0.9>
    gravityOverride {z=-980}
    physicsBodyDefinitions [{
      collisionType=CustomSphere
      sphereCollisionRadius=2
      boundBone {boneName=stripe_r_04}
      boxExtents=<11 11 11>
      constraintSetup {angularLimitsMax=<0 3 15>, angularLimitsMin=<0 -3 -15>, angularTarget=<0 0 0>}
      localJointOffset=<0.7 0 0>
     }]
    planarLimits [{drivingBone {boneName=root}, planeTransform {location=<0 0 0>, rotation=<rot 0 0 0>, scale=<2 2 2>}}]
    simSpaceSettings {maxAngularAcceleration=350, maxAngularVelocity=30, simSpaceAngularAlpha=0.25}
   }
   Output Pose(Result=Component To Local(ComponentPose=@n59)) {layerGroup=None}
   Save cached pose 'PreSim'(Pose=Copy Pose From Mesh() {bUseAttachedParent=true}) {cacheName=PreSim}
   n60 = Control Rig!(
    Source=Use cached pose 'PreSim'()
    Alpha
    IsSlidingKnee:bool=Get bIsKneeSliding()
    IsSlidingBaseball:bool=Get bIsBaseballSliding()
    IsParachuteOpenedAndCharUpRight:bool=Get bParachuteOpenedAndStandingUpright() -> (bParachuteOpenedAndStandingUpright:bool)
    isDBNO:bool=Get bIsDBNO()
    isParachuteOpen:bool=Get bIsParachuteOpened()
    isSkydiveFloating:bool=Get bIsSkydiveFloating()
   ) -> (Pose) {controlRigAssetReference {blueprintRigClass=&/M_MED_RankedPhantom_cape.M_MED_RankedPhantom_cape_C}}
   Save cached pose 'Control RigDef'(Pose=@n60) {cacheName=Control RigDef}
  ]}
 EventGraph {nodes [
   Event BlueprintUpdateAnimation()
   > Sequence!(execute) -> (then_0, then_1, then_2:Exec)
   > n1 = Gravity Override State Function DEPRECATED(
    CharacterPartAnimInstance=Self-Reference()
    GameTimeInSeconds=GetGameTimeInSeconds()
    GravityOverrideParams=Get Gravity Override Params()
    RBAN Params=Get RBAN Params()
   )
   > Set Gravity Override(Gravity Override=@n1)
  ]}
]
variables [
 Character Part Anim Instance {type=CustomCharacterPartAnimInstance, default=null}
 Game Time in Seconds {type=number}
 Gravity Override {type=Vector}
 Gravity Override Params {
  type=GravityOverrideParamsStruct
  default {
   skydive_FloatingDown {gravityOverride=<0 -700 500>, windAmplitude=<200 -200 100>, windFrequency=<1.5 7 6>, windOffset=<0 -900 0>}
   skydive_GoingUp {gravityOverride=<0 -700 450>, windAmplitude=<200 -200 100>, windOffset=<0 -900 0>}
   skydive_Parachuting {gravityOverride=<0 -700 -980>, windAmplitude=<100 150 400>, windFrequency=<1.5 4 8>, windOffset=<0 500 0>}
  }
 }
 RBAN Params {
  type=RBANParamsStruct
  default {
   dBNO {simSpaceSettings {
     maxAngularAcceleration=1000
     maxAngularVelocity=3
     maxLinearAcceleration=300
     maxLinearVelocity=700
     velocityScaleZ=0
     externalLinearDragV=<0 0.5 0>
    }}
   dBNO_Carried {simSpaceSettings {
     maxAngularAcceleration=1000
     maxAngularVelocity=3
     maxLinearAcceleration=300
     maxLinearVelocity=700
     velocityScaleZ=0
     externalLinearDragV=<0 0.5 0>
    }}
   emoteOrMelee {simSpaceSettings {
     maxAngularAcceleration=1000
     maxAngularVelocity=3
     maxLinearAcceleration=300
     maxLinearVelocity=700
     velocityScaleZ=0
     externalLinearDragV=<0 0.5 0>
    }}
   front_End {simSpaceSettings {
     maxAngularAcceleration=1000
     maxAngularVelocity=3
     maxLinearAcceleration=300
     maxLinearVelocity=700
     velocityScaleZ=0
     externalLinearDragV=<0 0.5 0>
    }}
   onGround_Crouch_Moving {simSpaceSettings {
     maxAngularAcceleration=1000
     maxAngularVelocity=3
     maxLinearAcceleration=300
     maxLinearVelocity=700
     velocityScaleZ=0
     externalLinearDragV=<0 0.5 0>
    }}
   onGround_Falling {simSpaceSettings {
     maxAngularAcceleration=1000
     maxAngularVelocity=3
     maxLinearAcceleration=300
     maxLinearVelocity=700
     velocityScaleZ=0
     externalLinearDragV=<0 0.5 0>
    }}
   onGround_Moving {simSpaceSettings {
     maxAngularAcceleration=1000
     maxAngularVelocity=3
     maxLinearAcceleration=300
     maxLinearVelocity=700
     velocityScaleZ=0
     externalLinearDragV=<0 0.5 0>
    }}
   onGround_Standing {simSpaceSettings {
     maxAngularAcceleration=1000
     maxAngularVelocity=3
     maxLinearAcceleration=300
     maxLinearVelocity=700
     velocityScaleZ=0
     externalLinearDragV=<0 0.5 0>
    }}
   rail_Grinding {simSpaceSettings {
     maxAngularAcceleration=1000
     maxAngularVelocity=3
     maxLinearAcceleration=300
     maxLinearVelocity=700
     velocityScaleZ=0
     worldAlpha=0.1
     externalLinearDragV=<0 0.5 0>
    }}
   riding_Moving {simSpaceSettings {
     maxAngularAcceleration=1000
     maxAngularVelocity=3
     maxLinearAcceleration=300
     maxLinearVelocity=700
     velocityScaleZ=0
     externalLinearDragV=<0 0.5 0>
    }}
   riding_Moving_Targeting {simSpaceSettings {
     maxAngularAcceleration=1000
     maxAngularVelocity=3
     maxLinearAcceleration=300
     maxLinearVelocity=700
     velocityScaleZ=0
     externalLinearDragV=<0 0.5 0>
    }}
   riding_Standing {simSpaceSettings {
     maxAngularAcceleration=1000
     maxAngularVelocity=3
     maxLinearAcceleration=300
     maxLinearVelocity=700
     velocityScaleZ=0
     externalLinearDragV=<0 0.5 0>
    }}
   skydive_DivingDown {simSpaceSettings {
     dampingAlpha=0
     maxAngularAcceleration=1000
     maxAngularVelocity=3
     maxLinearAcceleration=300
     maxLinearVelocity=700
     velocityScaleZ=0
     externalLinearDragV=<0 0.5 0>
    }}
   skydive_FloatingDown {
    componentLinearAccScale=<0 0.5 0>
    componentLinearVelScale=<0 0.5 0>
    simSpaceSettings {
     dampingAlpha=0
     maxAngularAcceleration=1000
     maxAngularVelocity=3
     maxLinearAcceleration=300
     maxLinearVelocity=700
     velocityScaleZ=0
     externalLinearDragV=<0 0.5 0>
    }
   }
   skydive_GoingUp {simSpaceSettings {
     maxAngularAcceleration=1000
     maxAngularVelocity=3
     maxLinearAcceleration=300
     maxLinearVelocity=700
     velocityScaleZ=0
     externalLinearDragV=<0 0.5 0>
    }}
   skydive_Parachuting {
    componentLinearAccScale=<0 0.1 0.3>
    componentLinearVelScale=<0 0.1 0.3>
    simSpaceSettings {
     maxAngularAcceleration=1000
     maxAngularVelocity=3
     maxLinearAcceleration=300
     maxLinearVelocity=700
     velocityScaleZ=0
     externalLinearDragV=<0 0.5 0>
    }
   }
   skydive_Surfing {simSpaceSettings {
     maxAngularAcceleration=1000
     maxAngularVelocity=3
     maxLinearAcceleration=300
     maxLinearVelocity=700
     velocityScaleZ=0
     externalLinearDragV=<0 0.5 0>
    }}
   swimming {simSpaceSettings {
     maxAngularAcceleration=1000
     maxAngularVelocity=3
     maxLinearAcceleration=300
     maxLinearVelocity=700
     velocityScaleZ=0
     externalLinearDragV=<0 0.5 0>
    }}
   tactical_Sprint {simSpaceSettings {
     maxAngularAcceleration=1000
     maxAngularVelocity=3
     maxLinearAcceleration=300
     maxLinearVelocity=700
     velocityScaleZ=0
     externalLinearDragV=<0 0.5 0>
    }}
   ziplining {simSpaceSettings {
     maxAngularAcceleration=1000
     maxAngularVelocity=3
     maxLinearAcceleration=300
     maxLinearVelocity=700
     velocityScaleZ=0
     externalLinearDragV=<0 0.5 0>
    }}
  }
 }
]
