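# Physics asset for the dangling coat strips on the male medium RankedPhantom player cosmetic body.
# Capsules on pelvis, calves, feet and coat panels, plus a wide flat box at the root, form a mostly locked framework.
# Short stripe chains on back, front and sides have angle-limited swing joints with little or no twist.
# Exceptions are two fully free left-front coat joints and a left side stripe base without any parent constraint.
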PhysicsAsset /BRCosmetics/Characters/Player/Male/Medium/Bodies/M_MED_RankedPhantom/Meshes/M_MED_RankedPhantom_Strips
bodies {
 calf_l [{shapeType=Capsule, length=30, radius=7, shapeName=calf_l_capsule, center=<-22 -3 0>, rotation=<rot 90 -0.0002 0>}]
 calf_r [{shapeType=Capsule, length=30, radius=8, shapeName=calf_r_capsule, center=<22 3 -0.9037>, rotation=<rot 90 -0.0002 0>}]
 coat_bck_02 [
  {shapeType=Capsule, length=30, radius=6, shapeName=coat_bck_02_capsule, center=<0.5125 4.6823 0>}
  {shapeType=Capsule, length=30, radius=3, shapeName=coat_bck_02_capsule, center=<6.3814 2.0877 0>}
 ]
 coat_bck_03 [
  {shapeType=Capsule, length=30, radius=6, shapeName=coat_bck_03_capsule, center=<-1.044 4.8067 0>}
  {shapeType=Capsule, length=30, radius=6, shapeName=coat_bck_03_capsule, center=<-11.3032 4.7313 0>}
 ]
 coat_fwd1_l_01 [{shapeType=Capsule, length=0.5, radius=0.5, shapeName=coat_fwd1_l_01_capsule, rotation=<rot 0 0 -180>}]
 coat_fwd1_l_02 [
  {shapeType=Capsule, length=15, radius=3, shapeName=coat_fwd1_l_02_capsule, center=<15.8393 3.142 -3.4015>, rotation=<rot 0 0 -180>}
  {shapeType=Capsule, length=15, radius=3, shapeName=coat_fwd1_l_02_capsule, center=<11.8499 2.8904 -3.4015>, rotation=<rot 0 0 -180>}
 ]
 coat_fwd1_l_03 [
  {
   shapeType=Capsule
   length=15
   radius=3
   shapeName=coat_fwd1_l_03_capsule
   center=<-3.8001 2.7353 -4.289>
   rotation=<rot -0.0001 -0.0002 -171.5625>
  }
  {
   shapeType=Capsule
   length=15
   radius=3
   shapeName=coat_fwd1_l_03_capsule
   center=<0.7075 2.4126 -4.2555>
   rotation=<rot -0.0001 -0.0002 -171.5625>
  }
 ]
 coat_fwd1_l_base [{shapeType=Capsule, length=0.5, radius=0.5, shapeName=coat_fwd1_l_base_capsule, rotation=<rot 0 0 -180>}]
 coat_fwd1_r_01 [{shapeType=Capsule, length=0.5, radius=0.5, shapeName=coat_fwd1_r_01_capsule}]
 coat_fwd1_r_02 [
  {shapeType=Capsule, length=15, radius=3, shapeName=coat_fwd1_r_02_capsule, center=<9.8374 3.4928 2.3605>}
  {shapeType=Capsule, length=15, radius=3, shapeName=coat_fwd1_r_02_capsule, center=<14.4509 3.2957 2.3605>}
 ]
 coat_fwd1_r_03 [
  {
   shapeType=Capsule
   length=15
   radius=3
   shapeName=coat_fwd1_r_03_capsule
   center=<1.0071 2.5274 2.0371>
   rotation=<rot -0.0001 -0.0002 8.4375>
  }
  {
   shapeType=Capsule
   length=15
   radius=3
   shapeName=coat_fwd1_r_03_capsule
   center=<-4.0293 3.5836 2.3137>
   rotation=<rot -0.0001 -0.0002 8.4375>
  }
 ]
 coat_fwd1_r_base [{shapeType=Capsule, length=0.5, radius=0.5, shapeName=coat_fwd1_r_base_capsule}]
 coat_l_01 [{shapeType=Capsule, length=14, radius=4, shapeName=coat_l_01_capsule, center=<12.6291 0 0>}]
 coat_l_02 [{shapeType=Capsule, length=40.4231, radius=8.9756, center=<6.1502 4.6756 -2.5347>, rotation=<rot 2.1744 1.7855 36.5962>}]
 coat_l_03 [{shapeType=Capsule, length=32.0857, radius=9.9639, center=<-3.0499 4.1419 -2.0236>, rotation=<rot 0.0005 0.0003 36.5624>}]
 coat_l_base [{shapeType=Capsule, length=14, radius=4, shapeName=coat_l_base_capsule}]
 coat_r_01 [{shapeType=Capsule, length=14, radius=4, shapeName=coat_r_01_capsule, center=<10.0711 2.0708 0.1905>}]
 coat_r_02 [{
   shapeType=Capsule
   length=40.4231
   radius=8.9756
   shapeName=coat_r_02_capsule
   center=<-0.5958 9.6371 -0.7805>
   rotation=<rot -5.7746 2.1513 -32.4074>
  }]
 coat_r_03 [{
   shapeType=Capsule
   length=32.0857
   radius=9.9639
   shapeName=coat_r_03_capsule
   center=<-7.9854 9.715 -1.543>
   rotation=<rot -5.6363 2.0535 -35.1188>
  }]
 coat_r_base [{shapeType=Capsule, length=14, radius=4, shapeName=coat_r_base_capsule}]
 foot_l [{shapeType=Capsule, length=15, radius=8, shapeName=foot_l_capsule, center=<-5.5 6 0>, rotation=<rot 0 -0.0002 90>}]
 foot_r [{shapeType=Capsule, length=15, radius=8, shapeName=foot_r_capsule, center=<5.5 -6 0>, rotation=<rot 0 -0.0002 90>}]
 pelvis [{shapeType=Capsule, length=12, radius=10, shapeName=pelvis_capsule}]
 root [{shapeType=Box, extentX=200, extentY=200, extentZ=25, shapeName=root_box, center=<0 0 -8.9959>, rotation=<rot 0 -90 0>}]
 stripe_bck_m_l_01 [{shapeType=Capsule, length=7, radius=2.5, shapeName=stripe_bck_m_l_01_capsule, center=<6.7298 -2.1292 0>, rotation=<rot 90 5.6256 0>}]
 stripe_bck_m_l_02 [{shapeType=Capsule, length=7, radius=2.5, shapeName=stripe_bck_m_l_02_capsule, center=<6.9404 -1.0973 0>, rotation=<rot 90 0.0003 0>}]
 stripe_bck_m_l_base [{
   shapeType=Capsule
   length=0.505
   radius=0.505
   shapeName=stripe_bck_m_l_base_capsule
   center=<-3.0496 0 0>
   rotation=<rot -0.0001 0.0011 -0.0001>
  }]
 stripe_bck_m_r_01 [{shapeType=Capsule, length=7, radius=2.5, shapeName=stripe_bck_m_r_01_capsule, center=<6.7298 -2.1292 0>, rotation=<rot 90 5.6264 0>}]
 stripe_bck_m_r_02 [{shapeType=Capsule, length=7, radius=2.5, shapeName=stripe_bck_m_r_02_capsule, center=<6.9404 -1.0973 0>, rotation=<rot 90 0.0002 0>}]
 stripe_bck_m_r_base [{shapeType=Capsule, length=0.505, radius=0.505, shapeName=stripe_bck_m_r_base_capsule, center=<-3.0496 0 0>}]
 stripe_fwd1_l_01 [{shapeType=Capsule, length=5, radius=2, shapeName=stripe_fwd1_l_01_capsule, center=<6.0026 -1.7772 0>, rotation=<rot 90 -0.0002 0>}]
 stripe_fwd1_l_02 [{shapeType=Capsule, length=5, radius=2, shapeName=stripe_fwd1_l_02_capsule, center=<2.4728 -1.44 0>, rotation=<rot 90 -0.0001 0>}]
 stripe_fwd1_r_01 [{shapeType=Capsule, length=5, radius=2, shapeName=stripe_fwd1_r_01_capsule, center=<3.8095 -2 0>, rotation=<rot 90 -0.0008 0>}]
 stripe_fwd1_r_02 [{shapeType=Capsule, length=5, radius=2, shapeName=stripe_fwd1_r_02_capsule, center=<0 -1.85 0>, rotation=<rot 90 -0.0005 0>}]
 stripe_l_01 [{shapeType=Capsule, length=12, radius=3, center=<0 -2.7574 0>, rotation=<rot 90 -5.6249 0>}]
 stripe_l_02 [{shapeType=Capsule, length=12, radius=3, center=<0 -2.81 0>, rotation=<rot 90 0.0002 0>}]
 stripe_l_03 [{shapeType=Capsule, length=12, radius=3, center=<0 -2 0>, rotation=<rot 90 0.0002 0>}]
 stripe_l_04 [{shapeType=Capsule, length=12, radius=3, rotation=<rot 90 0.0002 0>}]
 stripe_l_base [{shapeType=Capsule, length=12, radius=3, shapeName=stripe_l_base_capsule, center=<0 3.3294 0>, rotation=<rot 0 90 90>}]
 stripe_r_01 [{shapeType=Capsule, length=12, radius=3, shapeName=stripe_r_01_capsule, center=<0 -2.4142 0>, rotation=<rot -90 0 0>}]
 stripe_r_02 [{shapeType=Capsule, length=12, radius=3, shapeName=stripe_r_02_capsule, center=<0 -2.6807 0>, rotation=<rot -90 -5.625 0>}]
 stripe_r_03 [{shapeType=Capsule, length=12, radius=3, shapeName=stripe_r_03_capsule, center=<0 -2.6369 0>, rotation=<rot -90 0 0>}]
 stripe_r_04 [{shapeType=Capsule, length=12, radius=3, shapeName=stripe_r_04_capsule, rotation=<rot -90 0 0>}]
 stripe_r_base [{shapeType=Capsule, length=12, radius=3, shapeName=stripe_r_base_capsule, rotation=<rot -90 0 0>}]
}
constraints [
 {
  bone1Name=foot_r
  bone2Name=calf_r
  swing1LimitDegrees=45
  swing1Motion=Locked
  swing2LimitDegrees=45
  swing2Motion=Locked
  twistLimitDegrees=45
  twistMotion=Locked
 }
 {
  bone1Name=foot_l
  bone2Name=calf_l
  swing1LimitDegrees=45
  swing1Motion=Locked
  swing2LimitDegrees=45
  swing2Motion=Locked
  twistLimitDegrees=45
  twistMotion=Locked
 }
 {
  bone1Name=calf_l
  bone2Name=pelvis
  swing1LimitDegrees=45
  swing1Motion=Locked
  swing2LimitDegrees=45
  swing2Motion=Locked
  twistLimitDegrees=45
  twistMotion=Locked
 }
 {
  bone1Name=calf_r
  bone2Name=pelvis
  swing1LimitDegrees=45
  swing1Motion=Locked
  swing2LimitDegrees=45
  swing2Motion=Locked
  twistLimitDegrees=45
  twistMotion=Locked
 }
 {
  bone1Name=pelvis
  bone2Name=root
  swing1LimitDegrees=45
  swing1Motion=Locked
  swing2LimitDegrees=45
  swing2Motion=Locked
  twistLimitDegrees=45
  twistMotion=Locked
 }
 {
  bone1Name=stripe_bck_m_l_base
  bone2Name=coat_bck_02
  swing1LimitDegrees=45
  swing1Motion=Locked
  swing2LimitDegrees=45
  swing2Motion=Locked
  twistLimitDegrees=0
  twistMotion=Locked
 }
 {bone1Name=stripe_bck_m_l_01, bone2Name=stripe_bck_m_l_base, swing1LimitDegrees=25, swing2LimitDegrees=10, twistLimitDegrees=0}
 {bone1Name=stripe_bck_m_l_02, bone2Name=stripe_bck_m_l_01, swing1LimitDegrees=4.5, swing2LimitDegrees=6, twistLimitDegrees=0}
 {
  bone1Name=stripe_bck_m_r_base
  bone2Name=coat_bck_02
  swing1LimitDegrees=45
  swing1Motion=Locked
  swing2LimitDegrees=45
  swing2Motion=Locked
  twistLimitDegrees=0
  twistMotion=Locked
 }
 {bone1Name=stripe_bck_m_r_01, bone2Name=stripe_bck_m_r_base, swing1LimitDegrees=25, swing2LimitDegrees=10, twistLimitDegrees=0}
 {bone1Name=stripe_bck_m_r_02, bone2Name=stripe_bck_m_r_01, swing1LimitDegrees=4.5, swing2LimitDegrees=6, twistLimitDegrees=0}
 {
  bone1Name=coat_bck_02
  bone2Name=pelvis
  swing1LimitDegrees=45
  swing1Motion=Locked
  swing2LimitDegrees=45
  swing2Motion=Locked
  twistLimitDegrees=45
  twistMotion=Locked
 }
 {
  bone1Name=coat_bck_03
  bone2Name=coat_bck_02
  swing1LimitDegrees=45
  swing1Motion=Locked
  swing2LimitDegrees=45
  swing2Motion=Locked
  twistLimitDegrees=45
  twistMotion=Locked
 }
 {bone1Name=stripe_l_01, bone2Name=stripe_l_base, swing1LimitDegrees=15, swing2LimitDegrees=6, twistLimitDegrees=1}
 {bone1Name=stripe_l_02, bone2Name=stripe_l_01, swing1LimitDegrees=15, swing2LimitDegrees=15, twistLimitDegrees=1}
 {bone1Name=stripe_l_03, bone2Name=stripe_l_02, swing1LimitDegrees=15, swing2LimitDegrees=15, twistLimitDegrees=1}
 {bone1Name=stripe_l_04, bone2Name=stripe_l_03, swing1LimitDegrees=15, swing2LimitDegrees=15, twistLimitDegrees=1}
 {
  bone1Name=coat_l_base
  bone2Name=pelvis
  swing1LimitDegrees=45
  swing1Motion=Locked
  swing2LimitDegrees=45
  swing2Motion=Locked
  twistLimitDegrees=45
  twistMotion=Locked
 }
 {
  bone1Name=coat_l_01
  bone2Name=coat_l_base
  swing1LimitDegrees=45
  swing1Motion=Locked
  swing2LimitDegrees=45
  swing2Motion=Locked
  twistLimitDegrees=45
  twistMotion=Locked
 }
 {
  bone1Name=coat_fwd1_r_03
  bone2Name=coat_fwd1_r_02
  swing1LimitDegrees=45
  swing1Motion=Locked
  swing2LimitDegrees=45
  swing2Motion=Locked
  twistLimitDegrees=45
  twistMotion=Locked
 }
 {
  bone1Name=coat_fwd1_r_02
  bone2Name=coat_fwd1_r_01
  swing1LimitDegrees=45
  swing1Motion=Locked
  swing2LimitDegrees=45
  swing2Motion=Locked
  twistLimitDegrees=45
  twistMotion=Locked
 }
 {
  bone1Name=coat_fwd1_r_01
  bone2Name=coat_fwd1_r_base
  swing1LimitDegrees=45
  swing1Motion=Locked
  swing2LimitDegrees=45
  swing2Motion=Locked
  twistLimitDegrees=45
  twistMotion=Locked
 }
 {
  bone1Name=coat_fwd1_r_base
  bone2Name=pelvis
  swing1LimitDegrees=45
  swing1Motion=Locked
  swing2LimitDegrees=45
  swing2Motion=Locked
  twistLimitDegrees=45
  twistMotion=Locked
 }
 {bone1Name=stripe_fwd1_r_01, bone2Name=coat_fwd1_r_02, swing1LimitDegrees=15, swing2LimitDegrees=6, twistLimitDegrees=2, twistMotion=Locked}
 {bone1Name=stripe_fwd1_r_02, bone2Name=stripe_fwd1_r_01, swing1LimitDegrees=15, swing2LimitDegrees=6, twistLimitDegrees=2}
 {bone1Name=stripe_fwd1_l_01, bone2Name=coat_fwd1_l_02, swing1LimitDegrees=15, swing2LimitDegrees=6, twistLimitDegrees=2}
 {bone1Name=stripe_fwd1_l_02, bone2Name=stripe_fwd1_l_01, swing1LimitDegrees=15, swing2LimitDegrees=6, twistLimitDegrees=2}
 {
  bone1Name=coat_fwd1_l_03
  bone2Name=coat_fwd1_l_02
  swing1LimitDegrees=45
  swing1Motion=Free
  swing2LimitDegrees=45
  swing2Motion=Free
  twistLimitDegrees=45
  twistMotion=Free
 }
 {
  bone1Name=coat_fwd1_l_02
  bone2Name=coat_fwd1_l_01
  swing1LimitDegrees=45
  swing1Motion=Free
  swing2LimitDegrees=45
  swing2Motion=Free
  twistLimitDegrees=45
  twistMotion=Free
 }
 {
  bone1Name=coat_fwd1_l_01
  bone2Name=coat_fwd1_l_base
  swing1LimitDegrees=45
  swing1Motion=Locked
  swing2LimitDegrees=45
  swing2Motion=Locked
  twistLimitDegrees=45
  twistMotion=Locked
 }
 {
  bone1Name=coat_fwd1_l_base
  bone2Name=pelvis
  swing1LimitDegrees=45
  swing1Motion=Locked
  swing2LimitDegrees=45
  swing2Motion=Locked
  twistLimitDegrees=45
  twistMotion=Locked
 }
 {
  bone1Name=stripe_r_base
  bone2Name=coat_r_base
  swing1LimitDegrees=15
  swing1Motion=Locked
  swing2LimitDegrees=15
  swing2Motion=Locked
  twistLimitDegrees=0
  twistMotion=Locked
 }
 {bone1Name=stripe_r_01, bone2Name=stripe_r_base, swing1LimitDegrees=15, swing2LimitDegrees=8, twistLimitDegrees=0}
 {bone1Name=stripe_r_02, bone2Name=stripe_r_01, swing1LimitDegrees=15, swing2LimitDegrees=15, twistLimitDegrees=0}
 {bone1Name=stripe_r_03, bone2Name=stripe_r_02, swing1LimitDegrees=15, swing2LimitDegrees=15, twistLimitDegrees=0}
 {bone1Name=stripe_r_04, bone2Name=stripe_r_03, swing1LimitDegrees=15, swing2LimitDegrees=15, twistLimitDegrees=0}
 {
  bone1Name=coat_r_base
  bone2Name=pelvis
  swing1LimitDegrees=45
  swing1Motion=Locked
  swing2LimitDegrees=45
  swing2Motion=Locked
  twistLimitDegrees=45
  twistMotion=Locked
 }
 {
  bone1Name=coat_r_01
  bone2Name=coat_r_base
  swing1LimitDegrees=45
  swing1Motion=Locked
  swing2LimitDegrees=45
  swing2Motion=Locked
  twistLimitDegrees=45
  twistMotion=Locked
 }
 {
  bone1Name=coat_r_02
  bone2Name=coat_r_01
  swing1LimitDegrees=45
  swing1Motion=Locked
  swing2LimitDegrees=45
  swing2Motion=Locked
  twistLimitDegrees=45
  twistMotion=Locked
 }
 {
  bone1Name=coat_r_03
  bone2Name=coat_r_02
  swing1LimitDegrees=45
  swing1Motion=Locked
  swing2LimitDegrees=45
  swing2Motion=Locked
  twistLimitDegrees=45
  twistMotion=Locked
 }
 {
  bone1Name=coat_l_03
  bone2Name=coat_l_02
  swing1LimitDegrees=45
  swing1Motion=Locked
  swing2LimitDegrees=45
  swing2Motion=Locked
  twistLimitDegrees=45
  twistMotion=Locked
 }
 {
  bone1Name=coat_l_02
  bone2Name=coat_l_01
  swing1LimitDegrees=45
  swing1Motion=Locked
  swing2LimitDegrees=45
  swing2Motion=Locked
  twistLimitDegrees=45
  twistMotion=Locked
 }
]
