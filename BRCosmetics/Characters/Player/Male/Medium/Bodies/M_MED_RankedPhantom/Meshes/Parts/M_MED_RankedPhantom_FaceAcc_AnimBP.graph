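# Anim Blueprint for the RankedPhantom male character's face accessory mesh, adding corrective and secondary motion to the hood.
# Copies the attached parent's pose through a hood Control Rig, then applies an additive left-hood bone offset weighted by FrontEndCorr.
# Spring-driven AnimDynamics then simulates the left and right hood bones relative to spine_05, with gravity overridden in sim space.
# The event graph enables that offset only in the front end when no emote plays.
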AnimBlueprint /BRCosmetics/Characters/Player/Male/Medium/Bodies/M_MED_RankedPhantom/Meshes/Parts/M_MED_RankedPhantom_FaceAcc_AnimBP
target_skeleton=&/M_MED_RankedPhantom_FaceAcc_Skeleton
graphs [
 AnimGraph {nodes [
   n1 = Control Rig(Source=Copy Pose From Mesh() {bUseAttachedParent=true}) {controlRigAssetReference {blueprintRigClass=&/M_MED_RankedPhantom_FaceAcc_hood.M_MED_RankedPhantom_FaceAcc_hood_C}}
   n2 = "Transform (Modify) Bone - Bone: cor_hood_l"(
    ComponentPose=Local To Component(LocalPose=@n1)
    Alpha=Get FrontEndCorr() -> (FrontEndCorr:float)
    Translation="-3,0,0"
    Rotation="-20,0,0"
    Scale="1,1,1"
   ) {
    rotationMode=BMM_Additive
    rotationSpace=BCS_BoneSpace
    translationMode=BMM_Additive
    boneToModify {boneName=cor_hood_l}
    rotation {pitch=-20}
    translation {x=-3}
   }
   n3 = Anim Dynamics - Bone: cor_hood_l(ComponentPose=@n2, Alpha=1, ExternalForce="0,0,0") {
    angularDampingOverride=0.999
    angularSpringConstant=200
    bAngularSpring=true
    bGravityOverrideInSimSpace=true
    bLinearSpring=true
    bOverrideAngularDamping=true
    bOverrideLinearDamping=true
    bUseGravityOverride=true
    linearDampingOverride=1
    linearSpringConstant=150
    simulationSpace=BoneRelative
    boundBone {boneName=cor_hood_l}
    componentAppliedLinearAccClamp=<3000 3000 3000>
    componentLinearAccScale=<0.3 0.3 0.5>
    componentLinearVelScale=<0.3 0.3 0.8>
    gravityOverride {x=-980}
    physicsBodyDefinitions [{
      boundBone {boneName=cor_hood_l}
      boxExtents=<10 10 10>
      constraintSetup {angularLimitsMax=<0 10 0>, angularLimitsMin=<0 -10 0>, angularTarget=<1 0 0>, linearAxesMax=<2 0 2>, linearAxesMin=<-1 0 0>}
      localJointOffset=<0 0 -1>
     }]
    relativeSpaceBone {boneName=spine_05}
    simSpaceSettings {maxAngularAcceleration=2000, maxAngularVelocity=2000, simSpaceAngularAlpha=0.1}
   }
   n4 = Anim Dynamics - Bone: cor_hood_r(ComponentPose=@n3, Alpha=1, ExternalForce="0,0,0") {
    angularDampingOverride=0.999
    angularSpringConstant=200
    bAngularSpring=true
    bGravityOverrideInSimSpace=true
    bLinearSpring=true
    bOverrideAngularDamping=true
    bOverrideLinearDamping=true
    bUseGravityOverride=true
    linearDampingOverride=1
    linearSpringConstant=150
    simulationSpace=BoneRelative
    boundBone {boneName=cor_hood_r}
    componentAppliedLinearAccClamp=<3000 3000 3000>
    componentLinearAccScale=<0.3 0.3 0.5>
    componentLinearVelScale=<0.3 0.3 0.8>
    gravityOverride {x=-980}
    physicsBodyDefinitions [{
      boundBone {boneName=cor_hood_r}
      boxExtents=<11 11 11>
      constraintSetup {angularLimitsMax=<0 10 0>, angularLimitsMin=<0 -10 0>, angularTarget=<1 0 0>, linearAxesMax=<4 0 2>, linearAxesMin=<-1 0 -1>}
      localJointOffset=<0 0 -1>
     }]
    relativeSpaceBone {boneName=spine_05}
    simSpaceSettings {maxAngularAcceleration=2000, maxAngularVelocity=2000, simSpaceAngularAlpha=0.1}
   }
   Output Pose(Result=Component To Local(ComponentPose=@n4)) {layerGroup=None}
  ]}
 EventGraph {nodes [
   Event BlueprintUpdateAnimation()
   n1 = NOT Boolean(A=Get bIsPlayingEmote())
   n2 = Get bIsInFrontEnd()
   > Set FrontEndCorr(FrontEndCorr:float="To Float (Boolean)"(InBool=AND Boolean(A=@n2, B=@n1))) -> (Output_Get:float)
  ]}
]
variables [
 FrontEndCorr {type=number}
 isFrontAndNoEmote {type=boolean}
]
